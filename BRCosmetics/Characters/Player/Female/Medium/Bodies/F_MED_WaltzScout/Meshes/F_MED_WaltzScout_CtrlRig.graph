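# Control Rig for the F_MED_WaltzScout character body that procedurally corrects skirt, strap, foot and upper-arm bones as the character moves.
# Spherical pose readers on both thighs set directional leg weights, which drive additive bone corrections.
# Smoothed gameplay states such as crouching, jumping, hang glider, DBNO and rail grinding gate further adjustments.
# Parent constraints average center skirt bones between their sides and tie skirt to thigh during grinding.
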ControlRigBlueprint /BRCosmetics/Characters/Player/Female/Medium/Bodies/F_MED_WaltzScout/Meshes/F_MED_WaltzScout_CtrlRig
graphs [
 Multiply_11_ContainedGraph {nodes [
   n1 = RigVMFunctionEntryNode()
   RigVMFunctionReturnNode(Result=&RigVMFunction_MathDoubleMul(B=@n1.C, A=&RigVMFunction_MathDoubleMul(A=@n1.A, B=@n1.B).Result).Result)
  ]}
 Multiply_11_ContainedGraph {nodes [
   n1 = RigVMFunctionEntryNode()
   RigVMFunctionReturnNode(Result=&RigVMFunction_MathDoubleMul(B=@n1.C, A=&RigVMFunction_MathDoubleMul(A=@n1.A, B=@n1.B).Result).Result)
  ]}
 Multiply_14_ContainedGraph {nodes [
   n1 = RigVMFunctionEntryNode()
   n2 = &RigVMFunction_MathDoubleMul(A=@n1.A, B=@n1.B)
   RigVMFunctionReturnNode(Result=&RigVMFunction_MathDoubleMul(B=@n1.D, A=&RigVMFunction_MathDoubleMul(B=@n1.C, A=@n2.Result).Result).Result)
  ]}
 Multiply_14_ContainedGraph {nodes [
   n1 = RigVMFunctionEntryNode()
   n2 = &RigVMFunction_MathDoubleMul(A=@n1.A, B=@n1.B)
   RigVMFunctionReturnNode(Result=&RigVMFunction_MathDoubleMul(B=@n1.D, A=&RigVMFunction_MathDoubleMul(B=@n1.C, A=@n2.Result).Result).Result)
  ]}
 New Function_ContainedGraph {nodes [
   n1 = RigVMFunctionEntryNode()
   RigVMFunction_MathQuaternionToEuler = &RigVMFunction_MathQuaternionToEuler(
    RotationOrder=ZYX
    Value=&RigUnit_GetRelativeTransformForItem(
     Child=(Type=Bone,Name="None")
     Parent=(Type=Bone,Name="None")
     bChildInitial=False
     bParentInitial=False
     Child.Name=@n1.Child
     Parent.Name=@n1.Parent
    ).RelativeTransform.Rotation
   )
   Remap = &RigVMFunction_MathDoubleRemap(
    TargetMaximum=1
    TargetMinimum=0
    Value=@RigVMFunction_MathQuaternionToEuler.Result.Z
    SourceMaximum=@n1.SourceMax
    SourceMinimum=@n1.SourceMin
   )
   > RigVMFunctionReturnNode(DeltaZ=@Remap.Result)
  ]}
 New Function_ContainedGraph {nodes [
   RigVMFunctionEntryNode()
   > RigVMFunction_Sequence = RigVMAggregateNode()
   n1 = &RigUnit_SphericalPoseReader(
    ActiveRegionScaleFactors=(PositiveWidth=1,NegativeWidth=1,PositiveHeight=1,NegativeHeight=1)
    ActiveRegionSize=0.05
    Debug=(bDrawDebug=false,bDraw2D=False,bDrawLocalAxes=False,DebugScale=25,DebugSegments=20,DebugThickness=0.25)
    DriverAxis=(X=1,Y=0,Z=0)
    DriverItem=(Type=Bone,Name="thigh_r")
    FalloffRegionScaleFactors=(PositiveWidth=1,NegativeWidth=1.1,PositiveHeight=1,NegativeHeight=1)
    FalloffSize=0.55
    FlipHeightScaling=true
    FlipWidthScaling=true
    OptionalParentItem=(Type=Bone,Name="None")
    RotationOffset=(X=-90,Y=0,Z=-20)
    ExecutePin=@RigVMFunction_Sequence.A
   )
   > n2 = &RigUnit_SphericalPoseReader(
    ActiveRegionScaleFactors=(PositiveWidth=1,NegativeWidth=1,PositiveHeight=1,NegativeHeight=1)
    ActiveRegionSize=0.05
    Debug=(bDrawDebug=false,bDraw2D=False,bDrawLocalAxes=False,DebugScale=25,DebugSegments=20,DebugThickness=0.25)
    DriverAxis=(X=1,Y=0,Z=0)
    DriverItem=(Type=Bone,Name="thigh_r")
    FalloffRegionScaleFactors=(PositiveWidth=1,NegativeWidth=1,PositiveHeight=1,NegativeHeight=1)
    FalloffSize=0.25
    FlipHeightScaling=false
    FlipWidthScaling=false
    OptionalParentItem=(Type=Bone,Name="None")
    RotationOffset=(X=-40,Y=30,Z=0)
   )
   > n3 = &RigUnit_SphericalPoseReader(
    ActiveRegionScaleFactors=(PositiveWidth=1,NegativeWidth=1,PositiveHeight=1,NegativeHeight=1)
    ActiveRegionSize=0.05
    Debug=(bDrawDebug=false,bDraw2D=False,bDrawLocalAxes=False,DebugScale=25,DebugSegments=20,DebugThickness=0.25)
    DriverAxis=(X=1,Y=0,Z=0)
    DriverItem=(Type=Bone,Name="thigh_r")
    FalloffRegionScaleFactors=(PositiveWidth=1,NegativeWidth=1.2,PositiveHeight=1,NegativeHeight=1)
    FalloffSize=0.55
    FlipHeightScaling=true
    FlipWidthScaling=true
    OptionalParentItem=(Type=Bone,Name="None")
    RotationOffset=(X=0,Y=20,Z=0)
   )
   > n4 = &RigUnit_SphericalPoseReader(
    ActiveRegionScaleFactors=(PositiveWidth=1,NegativeWidth=1,PositiveHeight=1,NegativeHeight=1)
    ActiveRegionSize=0.05
    Debug=(bDrawDebug=false,bDraw2D=False,bDrawLocalAxes=False,DebugScale=25,DebugSegments=20,DebugThickness=0.25)
    DriverAxis=(X=1,Y=0,Z=0)
    DriverItem=(Type=Bone,Name="thigh_r")
    FalloffRegionScaleFactors=(PositiveWidth=1,NegativeWidth=1,PositiveHeight=1,NegativeHeight=1.25)
    FalloffSize=0.2
    FlipHeightScaling=true
    FlipWidthScaling=true
    OptionalParentItem=(Type=Bone,Name="None")
    RotationOffset=(X=40,Y=20,Z=0)
   )
   > n5 = &RigUnit_SphericalPoseReader(
    ActiveRegionScaleFactors=(PositiveWidth=1,NegativeWidth=1,PositiveHeight=1,NegativeHeight=1)
    ActiveRegionSize=0.05
    Debug=(bDrawDebug=false,bDraw2D=false,bDrawLocalAxes=False,DebugScale=25,DebugSegments=20,DebugThickness=0.25)
    DriverAxis=(X=1,Y=0,Z=0)
    DriverItem=(Type=Bone,Name="thigh_r")
    FalloffRegionScaleFactors=(PositiveWidth=1,NegativeWidth=1,PositiveHeight=1,NegativeHeight=1)
    FalloffSize=0.55
    FlipHeightScaling=true
    FlipWidthScaling=true
    OptionalParentItem=(Type=Bone,Name="None")
    RotationOffset=(X=90,Y=0,Z=20)
   )
   > n6 = &RigUnit_SphericalPoseReader(
    ActiveRegionScaleFactors=(PositiveWidth=1,NegativeWidth=1,PositiveHeight=1,NegativeHeight=1)
    ActiveRegionSize=0.05
    Debug=(bDrawDebug=false,bDraw2D=false,bDrawLocalAxes=False,DebugScale=25,DebugSegments=20,DebugThickness=0.25)
    DriverAxis=(X=1,Y=0,Z=0)
    DriverItem=(Type=Bone,Name="thigh_r")
    FalloffRegionScaleFactors=(PositiveWidth=2,NegativeWidth=1.5,PositiveHeight=1,NegativeHeight=1)
    FalloffSize=0.22
    FlipHeightScaling=true
    FlipWidthScaling=true
    OptionalParentItem=(Type=Bone,Name="None")
    RotationOffset=(X=-130,Y=0,Z=0)
   )
   > n7 = &RigUnit_SphericalPoseReader(
    ActiveRegionScaleFactors=(PositiveWidth=1,NegativeWidth=1,PositiveHeight=1,NegativeHeight=1)
    ActiveRegionSize=0.05
    Debug=(bDrawDebug=false,bDraw2D=False,bDrawLocalAxes=False,DebugScale=25,DebugSegments=20,DebugThickness=0.25)
    DriverAxis=(X=-1,Y=0,Z=0)
    DriverItem=(Type=Bone,Name="thigh_l")
    FalloffRegionScaleFactors=(PositiveWidth=1,NegativeWidth=1.1,PositiveHeight=1,NegativeHeight=1)
    FalloffSize=0.55
    FlipHeightScaling=False
    FlipWidthScaling=False
    OptionalParentItem=(Type=Bone,Name="None")
    RotationOffset=(X=90,Y=0,Z=-20)
   )
   > n8 = &RigUnit_SphericalPoseReader(
    ActiveRegionScaleFactors=(PositiveWidth=1,NegativeWidth=1,PositiveHeight=1,NegativeHeight=1)
    ActiveRegionSize=0.05
    Debug=(bDrawDebug=false,bDraw2D=False,bDrawLocalAxes=False,DebugScale=25,DebugSegments=20,DebugThickness=0.25)
    DriverAxis=(X=-1,Y=0,Z=0)
    DriverItem=(Type=Bone,Name="thigh_l")
    FalloffRegionScaleFactors=(PositiveWidth=1,NegativeWidth=1,PositiveHeight=1,NegativeHeight=1)
    FalloffSize=0.25
    FlipHeightScaling=False
    FlipWidthScaling=False
    OptionalParentItem=(Type=Bone,Name="None")
    RotationOffset=(X=140,Y=30,Z=0)
   )
   > n9 = &RigUnit_SphericalPoseReader(
    ActiveRegionScaleFactors=(PositiveWidth=1,NegativeWidth=1,PositiveHeight=1,NegativeHeight=1)
    ActiveRegionSize=0.05
    Debug=(bDrawDebug=false,bDraw2D=False,bDrawLocalAxes=False,DebugScale=25,DebugSegments=20,DebugThickness=0.25)
    DriverAxis=(X=-1,Y=0,Z=0)
    DriverItem=(Type=Bone,Name="thigh_l")
    FalloffRegionScaleFactors=(PositiveWidth=1,NegativeWidth=1.12,PositiveHeight=1,NegativeHeight=1)
    FalloffSize=0.55
    FlipHeightScaling=False
    FlipWidthScaling=False
    OptionalParentItem=(Type=Bone,Name="None")
    RotationOffset=(X=180,Y=20,Z=0)
   )
   > n10 = &RigUnit_SphericalPoseReader(
    ActiveRegionScaleFactors=(PositiveWidth=1,NegativeWidth=1,PositiveHeight=1,NegativeHeight=1)
    ActiveRegionSize=0.05
    Debug=(bDrawDebug=false,bDraw2D=False,bDrawLocalAxes=False,DebugScale=25,DebugSegments=20,DebugThickness=0.25)
    DriverAxis=(X=-1,Y=0,Z=0)
    DriverItem=(Type=Bone,Name="thigh_l")
    FalloffRegionScaleFactors=(PositiveWidth=1,NegativeWidth=1,PositiveHeight=1,NegativeHeight=1.25)
    FalloffSize=0.2
    FlipHeightScaling=true
    FlipWidthScaling=false
    OptionalParentItem=(Type=Bone,Name="None")
    RotationOffset=(X=-140,Y=20,Z=0)
   )
   > n11 = &RigUnit_SphericalPoseReader(
    ActiveRegionScaleFactors=(PositiveWidth=1,NegativeWidth=1,PositiveHeight=1,NegativeHeight=1)
    ActiveRegionSize=0.05
    Debug=(bDrawDebug=false,bDraw2D=False,bDrawLocalAxes=False,DebugScale=25,DebugSegments=20,DebugThickness=0.25)
    DriverAxis=(X=-1,Y=0,Z=0)
    DriverItem=(Type=Bone,Name="thigh_l")
    FalloffRegionScaleFactors=(PositiveWidth=1,NegativeWidth=1,PositiveHeight=1,NegativeHeight=1)
    FalloffSize=0.55
    FlipHeightScaling=False
    FlipWidthScaling=False
    OptionalParentItem=(Type=Bone,Name="None")
    RotationOffset=(X=-90,Y=0,Z=20)
   )
   > n12 = &RigUnit_SphericalPoseReader(
    ActiveRegionScaleFactors=(PositiveWidth=1,NegativeWidth=1,PositiveHeight=1,NegativeHeight=1)
    ActiveRegionSize=0.05
    Debug=(bDrawDebug=false,bDraw2D=False,bDrawLocalAxes=False,DebugScale=25,DebugSegments=20,DebugThickness=0.25)
    DriverAxis=(X=-1,Y=0,Z=0)
    DriverItem=(Type=Bone,Name="thigh_l")
    FalloffRegionScaleFactors=(PositiveWidth=2,NegativeWidth=2,PositiveHeight=1,NegativeHeight=2)
    FalloffSize=0.2
    FlipHeightScaling=true
    FlipWidthScaling=true
    OptionalParentItem=(Type=Bone,Name="None")
    RotationOffset=(X=50,Y=0,Z=0)
   )
   Multiply_15 = RigVMAggregateNode(
    C=&RigVMFunction_MathDoubleRemap(SourceMaximum=1, SourceMinimum=0, TargetMaximum=0, TargetMinimum=1, Value=RigVMVariableNode() {variable=Crouching}.Value).Result
    A=RigVMVariableNode() {variable=Leg_Front_R}.Value
    B=RigVMVariableNode() {variable=Leg_Front_L}.Value
   )
   n13 = &RigVMFunction_AnimEvalRichCurve(
    Curve=(EditorCurveData=(Keys=((),(InterpMode=RCIM_Cubic,TangentMode=RCTM_Break,Time=0.208,Value=0.3054,ArriveTangent=1.0223,LeaveTangent=0.3085),(InterpMode=RCIM_Cubic,TangentMode=RCTM_User,Time=0.9982,Value=0.5622,ArriveTangent=0.0189,LeaveTangent=0.0189))))
    SourceMinimum=0
    TargetMinimum=0
    Value=@Multiply_15.Result
   )
   > &RigUnit_ModifyTransforms(
    ItemToModify="((Item=(Type=Bone,Name=\"base_skirt_fr_r\"),Transform=(Rotation=(X=0,Y=0.3949,Z=0,W=0.9187),Translation=(X=0,Y=0,Z=0),Scale3D=(X=1,Y=1,Z=1))),(Item=(Type=Bone,Name=\"base_skirt_sd_r\"),Transform=(Rotation=(X=0,Y=0,Z=0,W=1),Translation=(X=0,Y=0,Z=0),Scale3D=(X= ...(truncated)"
    WeightMinimum=0
    Weight=@n13.Result
   )
   > &RigUnit_ModifyTransforms(
    ItemToModify=((Item=(Type=Bone,Name="base_skirt_fr_c"),Transform=(Rotation=(X=0,Y=0,Z=0,W=-1),Translation=(X=0,Y=0,Z=0),Scale3D=(X=1.3323,Y=1,Z=1))))
    WeightMinimum=0
    Weight=@Multiply_15.Result
   )
   Multiply_11 = RigVMAggregateNode(
    A=RigVMVariableNode() {variable=Leg_Front_R}.Value
    B=RigVMVariableNode() {variable=Leg_Front_L}.Value
    C=RigVMVariableNode() {variable=Crouching}.Value
   )
   n14 = &RigVMFunction_AnimEvalRichCurve(
    Curve=(EditorCurveData=(Keys=((),(InterpMode=RCIM_Cubic,Time=0.2661,Value=0.4716,ArriveTangent=1,LeaveTangent=1),(Time=1,Value=1))))
    SourceMinimum=0
    TargetMinimum=0
    Value=@Multiply_11.Result
   )
   > &RigUnit_ModifyTransforms(
    ItemToModify="((Item=(Type=Bone,Name=\"base_skirt_fr_r\"),Transform=(Rotation=(X=0,Y=0.2508,Z=-0.0001,W=0.968),Translation=(X=0,Y=0,Z=0),Scale3D=(X=1,Y=1,Z=1))),(Item=(Type=Bone,Name=\"base_skirt_fr_l\"),Transform=(Rotation=(X=0.0236,Y=0.1171,Z=-0.0127,W=0.9928),Translation ...(truncated)"
    WeightMinimum=0
    Weight=@n14.Result
   )
   > &RigUnit_ModifyTransforms(
    ItemToModify="((Item=(Type=Bone,Name=\"base_skirt_bk_r\"),Transform=(Rotation=(X=0.0003,Y=0.0202,Z=-0.0397,W=0.999),Translation=(X=0,Y=0,Z=0),Scale3D=(X=1,Y=1,Z=1))),(Item=(Type=Bone,Name=\"base_skirt_bk_l\"),Transform=(Rotation=(X=0.0015,Y=0.0401,Z=-0.0366,W=0.9985),Transl ...(truncated)"
    WeightMinimum=0
    Weight=&RigVMFunction_MathDoubleMul(B=RigVMVariableNode() {variable=Leg_Back_R}.Value, A=RigVMVariableNode() {variable=Leg_Back_L}.Value).Result
   )
   n15 = &RigVMFunction_AlphaInterp(
    InterpSpeedDecreasing=4
    InterpSpeedIncreasing=50
    Scale=1
    bClampResult=true
    Value=&RigVMFunction_MathBoolToFloat(Value=RigVMVariableNode() {variable=Jumping}.Value).Result
   )
   n16 = RigVMVariableNode() {variable=Leg_FrontInside_R}
   > &RigUnit_ModifyTransforms(
    ItemToModify="((Item=(Type=Bone,Name=\"base_skirt_fr_r\"),Transform=(Rotation=(X=0.1408,Y=0.1426,Z=-0.2055,W=0.9579),Translation=(X=0,Y=0,Z=0),Scale3D=(X=1,Y=1,Z=1))),(Item=(Type=Bone,Name=\"base_skirt_fr_c\"),Transform=(Rotation=(X=-0.0763,Y=0.0162,Z=0.282,W=-0.9562),Trans ...(truncated)"
    WeightMinimum=0
    Weight=&RigVMFunction_MathDoubleRemap(
     SourceMaximum=0.5
     SourceMinimum=0
     TargetMaximum=1
     TargetMinimum=0
     Value=&RigVMFunction_MathDoubleMul(B=@n15.Result, A=@n16.Value).Result
    ).Result
   )
   > &RigUnit_ModifyTransforms(
    ItemToModify="((Item=(Type=Bone,Name=\"base_strap_l\"),Transform=(Rotation=(X=0,Y=-0.0872,Z=0,W=0.9962),Translation=(X=0,Y=0,Z=0),Scale3D=(X=1,Y=1,Z=1))),(Item=(Type=Bone,Name=\"dyn_strap_fr_01_extra\"),Transform=(Rotation=(X=0,Y=0,Z=0,W=1),Translation=(X=0,Y=0,Z=0),Scale3D ...(truncated)"
    WeightMinimum=0
    Weight=&RigVMFunction_MathDoubleMul(A=@n15.Result, B=RigVMVariableNode() {variable=Leg_Front_L}.Value).Result
   )
   n17 = &RigVMFunction_AlphaInterp(InterpSpeedDecreasing=10, InterpSpeedIncreasing=10, Scale=1, bClampResult=true, Value=RigVMVariableNode() {variable=HangGlider}.Value)
   > &RigUnit_ModifyTransforms(
    ItemToModify=((Item=(Type=Bone,Name="base_strap_l"),Transform=(Rotation=(X=0,Y=0.0206,Z=0,W=-0.9998),Translation=(X=0,Y=0,Z=0),Scale3D=(X=1,Y=1,Z=1))))
    WeightMinimum=0
    Weight=@n17.Result
   )
   > n18 = &RigUnit_SphericalPoseReader(
    ActiveRegionScaleFactors=(PositiveWidth=1,NegativeWidth=1,PositiveHeight=1,NegativeHeight=1)
    ActiveRegionSize=0.05
    Debug=(bDrawDebug=false,bDraw2D=False,bDrawLocalAxes=False,DebugScale=25,DebugSegments=20,DebugThickness=0.25)
    DriverAxis=(X=1,Y=0,Z=0)
    DriverItem=(Type=Bone,Name="thigh_r")
    FalloffRegionScaleFactors=(PositiveWidth=1,NegativeWidth=0.7,PositiveHeight=1,NegativeHeight=1)
    FalloffSize=0.3
    FlipHeightScaling=true
    FlipWidthScaling=true
    OptionalParentItem=(Type=Bone,Name="None")
    RotationOffset=(X=-90,Y=0,Z=-40)
   )
   > &RigUnit_ModifyTransforms(
    ItemToModify="((Item=(Type=Bone,Name=\"base_skirt_fr_r\"),Transform=(Rotation=(X=0,Y=0.2272,Z=0,W=-0.9738),Translation=(X=0,Y=0,Z=0),Scale3D=(X=1,Y=1,Z=1))),(Item=(Type=Bone,Name=\"base_skirt_fr_c\"),Transform=(Rotation=(X=0.0842,Y=-0.3379,Z=0.0196,W=0.9372),Translation=(X= ...(truncated)"
    WeightMinimum=0
    Weight=@n18.OutputParam
   )
   > n19 = &RigUnit_SphericalPoseReader(
    ActiveRegionScaleFactors=(PositiveWidth=1,NegativeWidth=1,PositiveHeight=1,NegativeHeight=1)
    ActiveRegionSize=0.05
    Debug=(bDrawDebug=false,bDraw2D=False,bDrawLocalAxes=False,DebugScale=25,DebugSegments=20,DebugThickness=0.25)
    DriverAxis=(X=-1,Y=0,Z=0)
    DriverItem=(Type=Bone,Name="thigh_l")
    FalloffRegionScaleFactors=(PositiveWidth=1,NegativeWidth=0.7,PositiveHeight=1,NegativeHeight=1)
    FalloffSize=0.3
    FlipHeightScaling=false
    FlipWidthScaling=false
    OptionalParentItem=(Type=Bone,Name="None")
    RotationOffset=(X=90,Y=0,Z=-40)
   )
   > &RigUnit_ModifyTransforms(
    ItemToModify="((Item=(Type=Bone,Name=\"base_skirt_fr_l\"),Transform=(Rotation=(X=0,Y=-0.1181,Z=0,W=0.993),Translation=(X=-3.3583,Y=0,Z=0.5275),Scale3D=(X=1,Y=1,Z=1))),(Item=(Type=Bone,Name=\"base_skirt_fr_c\"),Transform=(Rotation=(X=0,Y=-0.3316,Z=0,W=0.9434),Translation=(X= ...(truncated)"
    WeightMinimum=0
    Weight=@n19.OutputParam
   )
   RigVMVariableNode(ExecuteContext=@RigVMFunction_Sequence.B, Value=@n1.OutputParam) {variable=Leg_Front_R}
   > RigVMVariableNode(Value=@n2.OutputParam) {variable=Leg_FrontSide_R}
   > RigVMVariableNode(Value=@n3.OutputParam) {variable=Leg_Side_R}
   > RigVMVariableNode(Value=@n4.OutputParam) {variable=Leg_BackSide_R}
   > RigVMVariableNode(Value=@n5.OutputParam) {variable=Leg_Back_R}
   > RigVMVariableNode(Value=@n6.OutputParam) {variable=Leg_FrontInside_R}
   > RigVMVariableNode(Value=@n7.OutputParam) {variable=Leg_Front_L}
   > RigVMVariableNode(Value=@n8.OutputParam) {variable=Leg_FrontSide_L}
   > RigVMVariableNode(Value=@n9.OutputParam) {variable=Leg_Side_L}
   > RigVMVariableNode(Value=@n10.OutputParam) {variable=Leg_BackSide_L}
   > RigVMVariableNode(Value=@n11.OutputParam) {variable=Leg_Back_L}
   > RigVMVariableNode(Value=@n12.OutputParam) {variable=Leg_FrontInside_L}
   &RigUnit_ModifyTransforms(
    ItemToModify="((Item=(Type=Bone,Name=\"base_skirt_fr_r\"),Transform=(Rotation=(X=-0.2122,Y=0.7046,Z=0.3777,W=-0.562),Translation=(X=-1.4588,Y=0.3088,Z=-0.0019),Scale3D=(X=0.8659,Y=1,Z=1))),(Item=(Type=Bone,Name=\"base_skirt_sd_r\"),Transform=(Rotation=(X=-0.0253,Y=0.1103,Z= ...(truncated)"
    WeightMinimum=0
    ExecutePin=@RigVMFunction_Sequence.C
    Weight=RigVMVariableNode() {variable=Leg_Front_R}.Value
   )
   > &RigUnit_ModifyTransforms(
    ItemToModify="((Item=(Type=Bone,Name=\"base_skirt_fr_r\"),Transform=(Rotation=(X=0,Y=0.2037,Z=0,W=-0.979),Translation=(X=0,Y=0,Z=0),Scale3D=(X=1,Y=1,Z=1))),(Item=(Type=Bone,Name=\"base_skirt_sd_r\"),Transform=(Rotation=(X=-0.2402,Y=0.03,Z=0.0887,W=-0.9662),Translation=(X=0, ...(truncated)"
    WeightMinimum=0
    Weight=RigVMVariableNode() {variable=Leg_FrontSide_R}.Value
   )
   > &RigUnit_ModifyTransforms(
    ItemToModify="((Item=(Type=Bone,Name=\"base_skirt_fr_r\"),Transform=(Rotation=(X=0.0683,Y=-0.2679,Z=0.1121,W=0.9545),Translation=(X=0,Y=0,Z=0),Scale3D=(X=0.9107,Y=1,Z=1))),(Item=(Type=Bone,Name=\"base_skirt_sd_r\"),Transform=(Rotation=(X=-0.0924,Y=0.6779,Z=0.0565,W=-0.7271) ...(truncated)"
    WeightMinimum=0
    Weight=RigVMVariableNode() {variable=Leg_Side_R}.Value
   )
   > &RigUnit_ModifyTransforms(
    ItemToModify="((Item=(Type=Bone,Name=\"base_skirt_fr_r\"),Transform=(Rotation=(X=0,Y=0,Z=0,W=1),Translation=(X=0,Y=0,Z=0),Scale3D=(X=1,Y=1,Z=1))),(Item=(Type=Bone,Name=\"base_skirt_sd_r\"),Transform=(Rotation=(X=0.0289,Y=-0.2165,Z=0.0764,W=0.9729),Translation=(X=0,Y=0,Z=0), ...(truncated)"
    WeightMinimum=0
    Weight=RigVMVariableNode() {variable=Leg_BackSide_R}.Value
   )
   > &RigUnit_ModifyTransforms(
    ItemToModify="((Item=(Type=Bone,Name=\"base_skirt_sd_r\"),Transform=(Rotation=(X=0,Y=0,Z=0.3928,W=0.9196),Translation=(X=0,Y=0,Z=0),Scale3D=(X=1,Y=1,Z=1))),(Item=(Type=Bone,Name=\"base_skirt_bk_r\"),Transform=(Rotation=(X=-0.1964,Y=-0.6896,Z=0.1685,W=0.6764),Translation=(X= ...(truncated)"
    WeightMinimum=0
    Weight=RigVMVariableNode() {variable=Leg_Back_R}.Value
   )
   n20 = RigVMVariableNode() {variable=Crouching}
   > &RigUnit_ModifyTransforms(
    ItemToModify="((Item=(Type=Bone,Name=\"base_skirt_fr_r\"),Transform=(Rotation=(X=0,Y=0.0158,Z=0,W=0.9999),Translation=(X=0,Y=0,Z=0),Scale3D=(X=1,Y=1,Z=1))),(Item=(Type=Bone,Name=\"base_skirt_fr_l\"),Transform=(Rotation=(X=0,Y=0,Z=0,W=1),Translation=(X=0,Y=0,Z=0),Scale3D=(X= ...(truncated)"
    WeightMinimum=0
    Weight=&RigVMFunction_MathDoubleMul(
     B=&RigVMFunction_MathDoubleRemap(SourceMaximum=1, SourceMinimum=0, TargetMaximum=0, TargetMinimum=1, Value=@n20.Value).Result
     A=RigVMVariableNode() {variable=Leg_FrontInside_R}.Value
    ).Result
   )
   > &RigUnit_ModifyTransforms(
    ItemToModify="((Item=(Type=Bone,Name=\"base_skirt_fr_l\"),Transform=(Rotation=(X=-0.3071,Y=-0.7644,Z=0.2068,W=0.5278),Translation=(X=0,Y=0,Z=0),Scale3D=(X=1,Y=1,Z=1))),(Item=(Type=Bone,Name=\"base_skirt_sd_l\"),Transform=(Rotation=(X=0.0498,Y=0.1139,Z=-0.3917,W=-0.9117),Tra ...(truncated)"
    WeightMinimum=0
    Weight=RigVMVariableNode() {variable=Leg_Front_L}.Value
   )
   > &RigUnit_ModifyTransforms(
    ItemToModify="((Item=(Type=Bone,Name=\"base_skirt_fr_l\"),Transform=(Rotation=(X=0.0211,Y=-0.093,Z=-0.0357,W=0.9948),Translation=(X=0,Y=0,Z=0),Scale3D=(X=1,Y=1,Z=1))),(Item=(Type=Bone,Name=\"base_skirt_sd_l\"),Transform=(Rotation=(X=-0.2377,Y=-0.1117,Z=0.0346,W=0.9643),Tran ...(truncated)"
    WeightMinimum=0
    Weight=RigVMVariableNode() {variable=Leg_FrontSide_L}.Value
   )
   > &RigUnit_ModifyTransforms(
    ItemToModify="((Item=(Type=Bone,Name=\"base_skirt_fr_l\"),Transform=(Rotation=(X=0.0932,Y=0.2491,Z=0.3443,W=-0.9004),Translation=(X=0,Y=0,Z=0),Scale3D=(X=1,Y=1,Z=1))),(Item=(Type=Bone,Name=\"base_skirt_sd_l\"),Transform=(Rotation=(X=0.0466,Y=0.6785,Z=-0.0502,W=-0.7314),Tran ...(truncated)"
    WeightMinimum=0
    Weight=RigVMVariableNode() {variable=Leg_Side_L}.Value
   )
   > &RigUnit_ModifyTransforms(
    ItemToModify="((Item=(Type=Bone,Name=\"base_skirt_fr_l\"),Transform=(Rotation=(X=0,Y=0,Z=0,W=1),Translation=(X=0,Y=0,Z=0),Scale3D=(X=1,Y=1,Z=1))),(Item=(Type=Bone,Name=\"base_skirt_sd_l\"),Transform=(Rotation=(X=-0.0477,Y=-0.0993,Z=-0.2173,W=0.9699),Translation=(X=0,Y=0,Z=0 ...(truncated)"
    WeightMinimum=0
    Weight=RigVMVariableNode() {variable=Leg_BackSide_L}.Value
   )
   > &RigUnit_ModifyTransforms(
    ItemToModify="((Item=(Type=Bone,Name=\"base_skirt_sd_l\"),Transform=(Rotation=(X=0,Y=0,Z=0,W=1),Translation=(X=0,Y=0,Z=0),Scale3D=(X=1,Y=1,Z=1))),(Item=(Type=Bone,Name=\"base_skirt_bk_c\"),Transform=(Rotation=(X=-0.0162,Y=-0.6905,Z=0.0168,W=0.723),Translation=(X=0,Y=0,Z=0), ...(truncated)"
    WeightMinimum=0
    Weight=RigVMVariableNode() {variable=Leg_Back_L}.Value
   )
   n21 = RigVMVariableNode() {variable=Leg_FrontInside_L}
   > &RigUnit_ModifyTransforms(
    ItemToModify="((Item=(Type=Bone,Name=\"base_skirt_fr_l\"),Transform=(Rotation=(X=0,Y=0,Z=0,W=1),Translation=(X=0,Y=0,Z=0),Scale3D=(X=1,Y=1,Z=1))),(Item=(Type=Bone,Name=\"base_skirt_fr_r\"),Transform=(Rotation=(X=-0.0955,Y=0.2325,Z=0.087,W=-0.964),Translation=(X=0,Y=0,Z=0),S ...(truncated)"
    WeightMinimum=0
    Weight=@n21.Value
   )
   n22 = &RigVMFunction_AlphaInterp(
    InterpSpeedDecreasing=10
    InterpSpeedIncreasing=10
    Scale=1
    bClampResult=true
    Value=&RigVMFunction_MathBoolToFloat(Value=RigVMVariableNode() {variable=DBNOBool}.Value).Result
   )
   > &RigUnit_ModifyTransforms(
    ItemToModify="((Item=(Type=Bone,Name=\"base_skirt_bk_c\"),Transform=(Rotation=(X=0,Y=-0.1236,Z=0,W=-0.9923),Translation=(X=0.1829,Y=0,Z=0.7229),Scale3D=(X=1,Y=1,Z=1))),(Item=(Type=Bone,Name=\"base_skirt_bk_l\"),Transform=(Rotation=(X=0,Y=-0.1451,Z=0,W=-0.9894),Translation=( ...(truncated)"
    WeightMinimum=0
    Weight=@n22.Result
   )
   n23 = &RigVMFunction_AlphaInterp(
    InterpSpeedDecreasing=10
    InterpSpeedIncreasing=10
    Scale=1
    bClampResult=true
    Value=&RigVMFunction_MathBoolToFloat(Value=RigVMVariableNode() {variable=DBNOCarried}.Value).Result
   )
   > &RigUnit_ModifyTransforms(
    ItemToModify="((Item=(Type=Bone,Name=\"base_strap_part_fr\"),Transform=(Rotation=(X=-0.0658,Y=0.0054,Z=-0.024,W=-0.9975),Translation=(X=0,Y=0,Z=0),Scale3D=(X=1,Y=1,Z=1))),(Item=(Type=Bone,Name=\"base_strap_part_bk\"),Transform=(Rotation=(X=-0.1576,Y=-0.0003,Z=-0.0138,W=0.98 ...(truncated)"
    WeightMinimum=0
    Weight=@n23.Result
   )
   RigVMFunctionReturnNode(ExecuteContext=@RigVMFunction_Sequence.E)
   Multiply = &RigVMFunction_MathDoubleMul(A=RigVMVariableNode() {variable=Leg_Front_L}.Value, B=RigVMVariableNode() {variable=Crouching}.Value)
   &RigUnit_ModifyTransforms(
    ItemToModify="((Item=(Type=Bone,Name=\"base_skirt_fr_l\"),Transform=(Rotation=(X=-0.0464,Y=-0.0272,Z=0.0223,W=-0.9983),Translation=(X=-0.873,Y=-0.9131,Z=0.3991),Scale3D=(X=1,Y=1,Z=1))),(Item=(Type=Bone,Name=\"base_skirt_sd_l\"),Transform=(Rotation=(X=0.2856,Y=0.094,Z=0.0568 ...(truncated)"
    WeightMinimum=0
    Weight=@Multiply.Result
    ExecutePin=@RigVMFunction_Sequence.D
   )
   > &RigUnit_ModifyTransforms(
    ItemToModify="((Item=(Type=Bone,Name=\"base_skirt_fr_r\"),Transform=(Rotation=(X=0.0765,Y=-0.0874,Z=-0.0307,W=0.9928),Translation=(X=0,Y=0,Z=0),Scale3D=(X=1,Y=1,Z=1))),(Item=(Type=Bone,Name=\"base_skirt_fr_c\"),Transform=(Rotation=(X=-0.4066,Y=0.2965,Z=0.0611,W=-0.862),Tran ...(truncated)"
    WeightMinimum=0
    Weight=&RigVMFunction_MathDoubleMul(B=RigVMVariableNode() {variable=Crouching}.Value, A=RigVMVariableNode() {variable=Leg_Front_R}.Value).Result
   )
   n24 = &RigVMFunction_AlphaInterp(
    InterpSpeedDecreasing=10
    InterpSpeedIncreasing=10
    Scale=1
    bClampResult=true
    Value=&RigVMFunction_MathDoubleMul(A=RigVMVariableNode() {variable=Idle}.Value, B=RigVMVariableNode() {variable=Crouching}.Value).Result
   )
   > &RigUnit_ModifyTransforms(
    ItemToModify=((Item=(Type=Bone,Name="base_skirt_bk_l"),Transform=(Rotation=(X=0,Y=0.1732,Z=0,W=0.9849),Translation=(X=0,Y=0,Z=0),Scale3D=(X=1,Y=1,Z=1))))
    WeightMinimum=0
    Weight=@n24.Result
   )
   > &RigUnit_ModifyTransforms(
    ItemToModify="((Item=(Type=Bone,Name=\"base_skirt_fr_c\"),Transform=(Rotation=(X=-0.1663,Y=0.1106,Z=0.4098,W=-0.89),Translation=(X=0,Y=0,Z=0),Scale3D=(X=1,Y=1,Z=1))),(Item=(Type=Bone,Name=\"base_skirt_fr_l\"),Transform=(Rotation=(X=-0.2518,Y=0.1788,Z=0.1221,W=-0.9432),Trans ...(truncated)"
    WeightMinimum=0
    Weight=&RigVMFunction_MathDoubleMul(B=RigVMVariableNode() {variable=Leg_FrontInside_R}.Value, A=RigVMVariableNode() {variable=Crouching}.Value).Result
   )
   > &RigUnit_ModifyTransforms(
    ItemToModify="((Item=(Type=Bone,Name=\"base_skirt_fr_l\"),Transform=(Rotation=(X=0.0009,Y=0.0063,Z=0.0414,W=0.9991),Translation=(X=0,Y=0,Z=0),Scale3D=(X=1,Y=1,Z=1))),(Item=(Type=Bone,Name=\"base_skirt_fr_r\"),Transform=(Rotation=(X=-0.514,Y=-0.1869,Z=-0.1147,W=0.8292),Trans ...(truncated)"
    WeightMinimum=0
    Weight=&RigVMFunction_MathDoubleMul(A=RigVMVariableNode() {variable=Leg_FrontInside_L}.Value, B=RigVMVariableNode() {variable=Crouching}.Value).Result
   )
   > DeltaRotationZ_1 = RigVMFunctionReferenceNode(Child=foot_l, Parent=calf_l, SourceMax=60, SourceMin=15)
   > &RigUnit_ModifyTransforms(
    ItemToModify=((Item=(Type=Bone,Name="foot_bk_l"),Transform=(Rotation=(X=0,Y=0,Z=0,W=-1),Translation=(X=0.2,Y=-5,Z=-5.62),Scale3D=(X=1,Y=1,Z=1))))
    WeightMinimum=0
    Weight=@DeltaRotationZ_1.DeltaZ
   )
   > DeltaRotationZ_2 = RigVMFunctionReferenceNode(Child=foot_r, Parent=calf_r, SourceMax=60, SourceMin=15)
   > &RigUnit_ModifyTransforms(
    ItemToModify=((Item=(Type=Bone,Name="foot_bk_r"),Transform=(Rotation=(X=0,Y=0,Z=0,W=-1),Translation=(X=-0.2,Y=5,Z=5.62),Scale3D=(X=1,Y=1,Z=1))))
    WeightMinimum=0
    Weight=@DeltaRotationZ_2.DeltaZ
   )
   RigVMFunction_MathBoolToFloat = &RigVMFunction_MathBoolToFloat(Value=RigVMVariableNode() {variable=GrindRail}.Value)
   AlphaInterp = &RigVMFunction_AlphaInterp(InterpSpeedDecreasing=10, InterpSpeedIncreasing=10, Scale=1, bClampResult=true, Value=@RigVMFunction_MathBoolToFloat.Result)
   > &RigUnit_ParentConstraint(
    AdvancedSettings=(InterpolationType=Average,RotationOrderForFilter=XZY)
    Child=(Type=Bone,Name="base_skirt_fr_l")
    Filter=(TranslationFilter=(bX=False,bY=False,bZ=False),RotationFilter=(bX=True,bY=True,bZ=True),ScaleFilter=(bX=True,bY=True,bZ=True))
    Parents=((Item=(Type=Bone,Name="thigh_l"),Weight=1))
    Weight=@AlphaInterp.Result
   )
   > &RigUnit_ParentConstraint(
    AdvancedSettings=(InterpolationType=Average,RotationOrderForFilter=XZY)
    Child=(Type=Bone,Name="base_skirt_sd_l")
    Filter=(TranslationFilter=(bX=False,bY=False,bZ=False),RotationFilter=(bX=True,bY=True,bZ=True),ScaleFilter=(bX=True,bY=True,bZ=True))
    Parents=((Item=(Type=Bone,Name="thigh_l"),Weight=1))
    Weight=@AlphaInterp.Result
   )
   > &RigUnit_ParentConstraint(
    AdvancedSettings=(InterpolationType=Average,RotationOrderForFilter=XZY)
    Child=(Type=Bone,Name="base_strap_part_fr")
    Parents=((Item=(Type=Bone,Name="dyn_skirt_fr_02_l"),Weight=1))
    Weight=@AlphaInterp.Result
   )
   > &RigUnit_ModifyTransforms(
    ItemToModify="((Item=(Type=Bone,Name=\"base_strap_part_fr\"),Transform=(Rotation=(X=-0.058,Y=0.2616,Z=0.0823,W=-0.9599),Translation=(X=0,Y=0,Z=0),Scale3D=(X=1,Y=1,Z=1))),(Item=(Type=Bone,Name=\"base_strap_part_bk\"),Transform=(Rotation=(X=0,Y=0,Z=-0.0387,W=0.9992),Translati ...(truncated)"
    WeightMinimum=0
    Weight=@AlphaInterp.Result
   )
   > DeltaRotationZ = RigVMFunctionReferenceNode(Child=thigh_l, Parent=pelvis, SourceMax=-150, SourceMin=-110)
   > &RigUnit_ModifyTransforms(
    ItemToModify="((Item=(Type=Bone,Name=\"base_skirt_fr_l\"),Transform=(Rotation=(X=-0.0196,Y=0.5004,Z=0.0708,W=-0.8627),Translation=(X=-6.6444,Y=-0.3841,Z=2.7277),Scale3D=(X=1,Y=1,Z=1))),(Item=(Type=Bone,Name=\"base_skirt_sd_l\"),Transform=(Rotation=(X=0,Y=0.1439,Z=0,W=-0.989 ...(truncated)"
    WeightMinimum=0
    Weight=&RigVMFunction_MathDoubleMul(B=@AlphaInterp.Result, A=@DeltaRotationZ.DeltaZ).Result
   )
   > DeltaRotationZ_3 = RigVMFunctionReferenceNode(Child=calf_r, Parent=thigh_r, SourceMax=100, SourceMin=150)
   RigUnit_GetRelativeTransformForItem = &RigUnit_GetRelativeTransformForItem(Child=(Type=Bone,Name="pelvis"), Parent=(Type=Bone,Name="root"), bChildInitial=False, bParentInitial=False)
   Remap = &RigVMFunction_MathDoubleRemap(
    SourceMaximum=50
    SourceMinimum=45
    TargetMaximum=0
    TargetMinimum=1
    Value=@RigUnit_GetRelativeTransformForItem.RelativeTransform.Translation.Z
   )
   n25 = &RigVMFunction_AlphaInterp(
    InterpSpeedDecreasing=10
    InterpSpeedIncreasing=10
    Scale=1
    bClampResult=true
    Value=&RigVMFunction_MathBoolToFloat(Value=RigVMVariableNode() {variable=PlayingEasternBloc}.Value).Result
   )
   Multiply_8 = RigVMAggregateNode(A=@DeltaRotationZ_3.DeltaZ, B=@Remap.Result, C=RigVMVariableNode() {variable=Leg_Front_L}.Value, D=@n25.Result)
   > &RigUnit_ModifyTransforms(
    ItemToModify="((Item=(Type=Bone,Name=\"base_skirt_fr_r\"),Transform=(Rotation=(X=0.0516,Y=0.4633,Z=-0.0688,W=0.882),Translation=(X=11.8872,Y=-2.5293,Z=10.3105),Scale3D=(X=1,Y=1,Z=1))),(Item=(Type=Bone,Name=\"base_skirt_sd_r\"),Transform=(Rotation=(X=-0.0189,Y=0.0027,Z=-0.14 ...(truncated)"
    WeightMinimum=0
    Weight=@Multiply_8.Result
   )
   > DeltaRotationZ_4 = RigVMFunctionReferenceNode(Child=calf_l, Parent=thigh_l, SourceMax=100, SourceMin=150)
   Multiply_14 = RigVMAggregateNode(A=@DeltaRotationZ_4.DeltaZ, B=RigVMVariableNode() {variable=Leg_Front_L}.Value, C=@Remap.Result, D=@n25.Result)
   > &RigUnit_ModifyTransforms(
    ItemToModify="((Item=(Type=Bone,Name=\"base_skirt_fr_l\"),Transform=(Rotation=(X=0.0052,Y=0.2206,Z=0.0231,W=0.9751),Translation=(X=6.4338,Y=-0.6207,Z=8.462),Scale3D=(X=1,Y=1,Z=1))),(Item=(Type=Bone,Name=\"base_skirt_sd_l\"),Transform=(Rotation=(X=-0.0756,Y=0.0109,Z=-0.1422, ...(truncated)"
    WeightMinimum=0
    Weight=@Multiply_14.Result
   )
   &RigVMFunction_MathDoubleMul(A=@n21.Value, B=RigVMVariableNode() {variable=Crouching}.Value)
   RigVMFunction_AlphaInterp = &RigVMFunction_AlphaInterp(InterpSpeedDecreasing=10, InterpSpeedIncreasing=10, Scale=1, Value=0, bClampResult=False, bInterpResult=False)
  ]}
 New Function_ContainedGraph {nodes [
   n1 = RigVMFunctionEntryNode()
   RigVMFunction_Sequence = &RigVMFunction_Sequence(ExecuteContext=@n1.ExecuteContext)
   &RigUnit_ParentConstraint(
    AdvancedSettings=(InterpolationType=Average,RotationOrderForFilter=XZY)
    Child=(Type=Bone,Name="dyn_skirt_fr_02_c")
    Filter=(TranslationFilter=(bX=False,bY=False,bZ=False),RotationFilter=(bX=True,bY=True,bZ=True),ScaleFilter=(bX=False,bY=False,bZ=False))
    Parents=((Item=(Type=Bone,Name="dyn_skirt_fr_02_l"),Weight=1),(Item=(Type=Bone,Name="dyn_skirt_fr_02_r"),Weight=1))
    Weight=1
    ExecutePin=@RigVMFunction_Sequence.A
   )
   > &RigUnit_ParentConstraint(
    AdvancedSettings=(InterpolationType=Average,RotationOrderForFilter=XZY)
    Child=(Type=Bone,Name="dyn_skirt_fr_03_c")
    Filter=(TranslationFilter=(bX=False,bY=False,bZ=False),RotationFilter=(bX=True,bY=True,bZ=True),ScaleFilter=(bX=False,bY=False,bZ=False))
    Parents=((Item=(Type=Bone,Name="dyn_skirt_fr_03_l"),Weight=1),(Item=(Type=Bone,Name="dyn_skirt_fr_03_r"),Weight=1))
    Weight=1
   )
   > &RigUnit_ParentConstraint(
    AdvancedSettings=(InterpolationType=Average,RotationOrderForFilter=XZY)
    Child=(Type=Bone,Name="dyn_skirt_fr_04_c")
    Filter=(TranslationFilter=(bX=False,bY=False,bZ=False),RotationFilter=(bX=True,bY=True,bZ=True),ScaleFilter=(bX=False,bY=False,bZ=False))
    Parents=((Item=(Type=Bone,Name="dyn_skirt_fr_04_l"),Weight=1),(Item=(Type=Bone,Name="dyn_skirt_fr_04_r"),Weight=1))
    Weight=1
   )
   > &RigUnit_ParentConstraint(
    AdvancedSettings=(InterpolationType=Average,RotationOrderForFilter=XZY)
    Child=(Type=Bone,Name="dyn_skirt_bk_02_c")
    Filter=(TranslationFilter=(bX=False,bY=False,bZ=False),RotationFilter=(bX=True,bY=True,bZ=True),ScaleFilter=(bX=False,bY=False,bZ=False))
    Parents=((Item=(Type=Bone,Name="dyn_skirt_bk_02_l"),Weight=1),(Item=(Type=Bone,Name="dyn_skirt_bk_02_r"),Weight=1))
    Weight=1
   )
   > &RigUnit_ParentConstraint(
    AdvancedSettings=(InterpolationType=Average,RotationOrderForFilter=XZY)
    Child=(Type=Bone,Name="dyn_skirt_bk_03_c")
    Filter=(TranslationFilter=(bX=False,bY=False,bZ=False),RotationFilter=(bX=True,bY=True,bZ=True),ScaleFilter=(bX=False,bY=False,bZ=False))
    Parents=((Item=(Type=Bone,Name="dyn_skirt_bk_03_l"),Weight=1),(Item=(Type=Bone,Name="dyn_skirt_bk_03_r"),Weight=1))
    Weight=1
   )
   > &RigUnit_ParentConstraint(
    AdvancedSettings=(InterpolationType=Average,RotationOrderForFilter=XZY)
    Child=(Type=Bone,Name="dyn_skirt_bk_04_c")
    Filter=(TranslationFilter=(bX=False,bY=False,bZ=False),RotationFilter=(bX=True,bY=True,bZ=True),ScaleFilter=(bX=False,bY=False,bZ=False))
    Parents=((Item=(Type=Bone,Name="dyn_skirt_bk_04_l"),Weight=1),(Item=(Type=Bone,Name="dyn_skirt_bk_04_r"),Weight=1))
    Weight=1
   )
   RigVMFunctionReturnNode(ExecuteContext=@RigVMFunction_Sequence.B)
  ]}
 New Function_ContainedGraph {nodes [
   n1 = RigVMFunctionEntryNode()
   RigVMFunction_Sequence = &RigVMFunction_Sequence(ExecuteContext=@n1.ExecuteContext)
   RigVMFunction_MathBoolToFloat = &RigVMFunction_MathBoolToFloat(Value=RigVMVariableNode() {variable=CrouchingBool}.Value)
   AlphaInterp = &RigVMFunction_AlphaInterp(InterpSpeedDecreasing=10, InterpSpeedIncreasing=10, Scale=1, bClampResult=true, Value=@RigVMFunction_MathBoolToFloat.Result)
   RigVMVariableNode(Value=@AlphaInterp.Result, ExecuteContext=@RigVMFunction_Sequence.A) {variable=Crouching}
   RigVMFunctionReturnNode(ExecuteContext=@RigVMFunction_Sequence.B)
  ]}
 New Function_ContainedGraph {nodes [
   n1 = RigVMFunctionEntryNode()
   Clamp = &RigVMFunction_MathIntClamp(Value=@n1.primAxis)
   RigVMFunction_MathBoolToInteger = &RigVMFunction_MathBoolToInteger(Value=@n1.onlyTwist)
   RigVMFunction_MathQuaternionSwingTwist = &RigVMFunction_MathQuaternionSwingTwist(
    TwistAxis=@n1.twistAcis
    Input=&RigUnit_GetRelativeTransformForItem(
     Child=(Type=Bone,Name="None")
     Parent=(Type=Bone,Name="None")
     bChildInitial=False
     bParentInitial=False
     Parent.Type=@n1.ParentType
     Child.Name=@n1.child
     Parent.Name=@n1.parent
    ).RelativeTransform.Rotation
   )
   RigVMFunction_MathQuaternionToEuler = &RigVMFunction_MathQuaternionToEuler(
    Value=RigVMDispatchNode(
     Values=((),())
     Index=@RigVMFunction_MathBoolToInteger.Result
     Values.0=@RigVMFunction_MathQuaternionSwingTwist.Swing
     Values.1=@RigVMFunction_MathQuaternionSwingTwist.Twist
    ) {template="DISPATCH_RigVMDispatch_SelectInt32(in Index,in Values,out Result)"}.Result
    RotationOrder=@n1.rotationOrder
   )
   DISPATCH_RigVMDispatch_SelectInt32 = RigVMDispatchNode(
    Values=(0,0,0)
    Index=@Clamp.Result
    Values.0=@RigVMFunction_MathQuaternionToEuler.Result.X
    Values.1=@RigVMFunction_MathQuaternionToEuler.Result.Y
    Values.2=@RigVMFunction_MathQuaternionToEuler.Result.Z
   ) {template="DISPATCH_RigVMDispatch_SelectInt32(in Index,in Values,out Result)"}
   Remap = &RigVMFunction_MathDoubleRemap(
    TargetMaximum=1
    TargetMinimum=0
    bClamp=False
    Value=@DISPATCH_RigVMDispatch_SelectInt32.Result
    SourceMaximum=@n1.max
    SourceMinimum=@n1.min
   )
   > RigVMFunctionReturnNode(OuputWeight=&RigVMFunction_MathDoubleClamp(Value=@Remap.Result).Result)
  ]}
 New Function_ContainedGraph {nodes [
   n1 = RigVMFunctionEntryNode()
   RigVMFunction_Sequence = &RigVMFunction_Sequence(ExecuteContext=@n1.ExecuteContext)
   AlphaInterp = &RigVMFunction_AlphaInterp(InterpSpeedDecreasing=10, InterpSpeedIncreasing=10, Scale=1, bClampResult=true, Value=RigVMVariableNode() {variable=HangGlider}.Value)
   &RigUnit_ModifyTransforms(
    ItemToModify="((Item=(Type=Bone,Name=\"deform_upperarm_up_l\"),Transform=(Rotation=(X=0,Y=-0.2164,Z=0,W=0.9763),Translation=(X=0,Y=0,Z=-2.5),Scale3D=(X=1,Y=1,Z=1))),(Item=(Type=Bone,Name=\"deform_upperarm_up_r\"),Transform=(Rotation=(X=0,Y=-0.2164,Z=0,W=0.9763),Translation= ...(truncated)"
    WeightMinimum=0
    Weight=@AlphaInterp.Result
    ExecutePin=@RigVMFunction_Sequence.A
   )
   n2 = &RigVMFunction_AlphaInterp(InterpSpeedDecreasing=10, InterpSpeedIncreasing=10, Scale=1, bClampResult=true, Value=RigVMVariableNode() {variable=FallingOnGround}.Value)
   > &RigUnit_ModifyTransforms(
    ItemToModify="((Item=(Type=Bone,Name=\"base_footStrap_in_l\"),Transform=(Rotation=(X=0,Y=-0.342,Z=0,W=0.9397),Translation=(X=0,Y=0,Z=0),Scale3D=(X=1,Y=1,Z=1))),(Item=(Type=Bone,Name=\"base_footStrap_out_l\"),Transform=(Rotation=(X=0,Y=-0.342,Z=0,W=0.9397),Translation=(X=0,Y ...(truncated)"
    WeightMinimum=0
    Weight=@n2.Result
   )
   > &RigUnit_ModifyTransforms(
    ItemToModify="((Item=(Type=Bone,Name=\"dyn_bow_02_l\"),Transform=(Rotation=(X=0,Y=-0.3827,Z=0,W=0.9239),Translation=(X=0,Y=0,Z=0.5),Scale3D=(X=1,Y=1,Z=1))),(Item=(Type=Bone,Name=\"dyn_bow_02_r\"),Transform=(Rotation=(X=0,Y=-0.3827,Z=0,W=0.9239),Translation=(X=0,Y=0,Z=-0.5), ...(truncated)"
    WeightMinimum=0
    Weight=@n2.Result
   )
   Multiply = &RigVMFunction_MathDoubleMul(
    A=RigVMDispatchNode(BoneName=root, Default=0, Name=deformRig_r_upperarm_up_alpha) {template="DISPATCH_RigDispatch_GetAnimAttribute(in Name,in BoneName,in Default,out Value,out Found)"}.Value
    B=RigVMVariableNode() {variable=Pistol}.Value
   )
   > &RigUnit_ModifyTransforms(
    ItemToModify="((Item=(Type=Bone,Name=\"deform_upperarm_fr_r\"),Transform=(Rotation=(X=0,Y=0,Z=0,W=-1),Translation=(X=24.451,Y=-3.8719,Z=0),Scale3D=(X=1,Y=1,Z=1))),(Item=(Type=Bone,Name=\"deform_upperarm_up_r\"),Transform=(Rotation=(X=0,Y=0,Z=0,W=1),Translation=(X=0,Y=1.3616 ...(truncated)"
    WeightMinimum=0
    Weight=@Multiply.Result
   )
   RigVMFunctionReturnNode(ExecuteContext=@RigVMFunction_Sequence.B)
  ]}
 New Function_ContainedGraph {nodes [
   n1 = RigVMFunctionEntryNode()
   RigVMFunction_Sequence = &RigVMFunction_Sequence(ExecuteContext=@n1.ExecuteContext)
   Multiply = &RigVMFunction_MathDoubleMul(A=RigVMVariableNode() {variable=Crouching}.Value, B=RigVMVariableNode() {variable=Idle}.Value)
   AlphaInterp = &RigVMFunction_AlphaInterp(InterpSpeedDecreasing=10, InterpSpeedIncreasing=10, Scale=1, bClampResult=true, Value=@Multiply.Result)
   &RigUnit_ModifyTransforms(
    ItemToModify="((Item=(Type=Bone,Name=\"base_strap_part_fr\"),Transform=(Rotation=(X=-0.1545,Y=0.0218,Z=0.0136,W=0.9877),Translation=(X=0.0983,Y=1.3537,Z=0.4058),Scale3D=(X=1,Y=1,Z=1))),(Item=(Type=Bone,Name=\"dyn_strap_part_fr_02\"),Transform=(Rotation=(X=0.0132,Y=-0.1676,Z ...(truncated)"
    WeightMinimum=0
    Weight=@AlphaInterp.Result
    ExecutePin=@RigVMFunction_Sequence.A
   )
   Remap = &RigVMFunction_MathDoubleRemap(SourceMaximum=1, SourceMinimum=0, TargetMaximum=0, TargetMinimum=1, Value=RigVMVariableNode() {variable=Idle}.Value)
   n2 = &RigVMFunction_AlphaInterp(
    InterpSpeedDecreasing=10
    InterpSpeedIncreasing=10
    Scale=1
    bClampResult=true
    Value=&RigVMFunction_MathDoubleMul(B=@Remap.Result, A=RigVMVariableNode() {variable=Crouching}.Value).Result
   )
   > &RigUnit_ModifyTransforms(
    ItemToModify="((Item=(Type=Bone,Name=\"base_strap_part_fr\"),Transform=(Rotation=(X=-0.156,Y=-0.0081,Z=-0.0511,W=0.9864),Translation=(X=0,Y=1.5243,Z=0),Scale3D=(X=1,Y=1,Z=1))),(Item=(Type=Bone,Name=\"dyn_strap_part_fr_02\"),Transform=(Rotation=(X=0,Y=0,Z=-0.143,W=-0.9897),T ...(truncated)"
    WeightMinimum=0
    Weight=@n2.Result
   )
   RigVMFunctionReturnNode(ExecuteContext=@RigVMFunction_Sequence.B)
  ]}
 RigVMFunctionLibrary {nodes [
   Additional Adjustments = RigVMCollapseNode(ExecuteContext=())
   ConstraintSetup = RigVMCollapseNode(ExecuteContext=())
   DeltaRotation = RigVMCollapseNode(ExecuteContext=(), onlyTwist=False, twistAcis=(X=0,Y=0,Z=0))
   DeltaRotationZ = RigVMCollapseNode(ExecuteContext=())
   PerformanceMode Setup = RigVMCollapseNode(ExecuteContext=())
   Set Variables = RigVMCollapseNode(ExecuteContext=())
   SkirtSetup = RigVMCollapseNode(ExecuteContext=())
  ]}
 RigVMFunction_Sequence_1_ContainedGraph {nodes [
   n1 = RigVMFunctionEntryNode()
   RigVMFunction_Sequence = &RigVMFunction_Sequence(ExecuteContext=@n1.ExecuteContext)
   n2 = &RigVMFunction_Sequence(ExecuteContext=@RigVMFunction_Sequence.B)
   n3 = &RigVMFunction_Sequence(ExecuteContext=@n2.B)
   n4 = &RigVMFunction_Sequence(ExecuteContext=@n3.B)
   RigVMFunctionReturnNode(A=@RigVMFunction_Sequence.A, B=@n2.A, C=@n3.A, D=@n4.A, E=@n4.B)
  ]}
 RigVMModel {nodes [
   n1 = &RigUnit_PrepareForExecution()
   &RigUnit_HierarchyImportFromSkeleton(NameSpace=None, bIncludeCurves=False, bIncludeMeshSockets=False, bIncludeVirtualBones=true, ExecutePin=@n1.ExecutePin)
   RigUnit_BeginExecution = &RigUnit_BeginExecution()
   FN_DeformRig_FNC = RigVMFunctionReferenceNode(
    DeformRig_ Curve="(deformRig_l_upperarm_up_alpha_72_#1=False,deformRig_r_upperarm_up_alpha_73_#2=true,deformRig_l_upperarm_fr_alpha_74_#3=False,deformRig_r_upperarm_fr_alpha_75_CB51B80 ...(truncated)"
    DeformRig_Bool="(upperarm_1_#4=True,elbow_3_#5=True,wrist_7_#6=True,groin_8_#7=True,glute_17_#8=True,knee_9_40C408324CA261 ...(truncated)"
    DeformRig_Debug=(upperarm_debug_18_#4=False,wrist_debug_20_#6=False,lat_debug_26_#9=False,groinGlute_debug_27_#7=False)
    DeformRig_Value="(upperarm_up_18_#10=(Rotation=(X=0,Y=0,Z=0,W=1),Translation=(X=0,Y=0,Z=2.5),Scale3D=(X=0.7,Y=1,Z=1)),upperarm_fwd_fr_19_#11=(Rotation=(X=0,Y=0,Z=0,W=1),Translation=(X=0,Y=10,Z=0),Scale3D=(X=0.5,Y=1, ...(truncated)"
    ExecuteContext=@RigUnit_BeginExecution.ExecutePin
   )
   > Set Variables = RigVMFunctionReferenceNode()
   > ConstraintSetup = RigVMFunctionReferenceNode()
   > SkirtSetup = RigVMFunctionReferenceNode()
   > Additional Adjustments = RigVMFunctionReferenceNode()
   > PerformanceMode Setup = RigVMFunctionReferenceNode()
  ]}
]
variables [
 Crouching {type=double, default=0}
 CrouchingBool {type=bool, default=False}
 DBNOBool {type=bool, default=False}
 DBNOCarried {type=bool, default=False}
 FallingOnGround {type=double, default=0}
 GrindRail {type=bool, default=False}
 HangGlider {type=double, default=0}
 Idle {type=double, default=0}
 Jumping {type=bool, default=False}
 Leg_BackSide_L {type=double, default=0}
 Leg_BackSide_R {type=double, default=0}
 Leg_Back_L {type=double, default=0}
 Leg_Back_R {type=double, default=0}
 Leg_FrontInside_L {type=double, default=0}
 Leg_FrontInside_R {type=double, default=0}
 Leg_FrontSide_L {type=double, default=0}
 Leg_FrontSide_R {type=double, default=0}
 Leg_Front_L {type=double, default=0}
 Leg_Front_R {type=double, default=0}
 Leg_Side_L {type=double, default=0}
 Leg_Side_R {type=double, default=0}
 Pistol {type=double, default=0}
 PlayingEasternBloc {type=bool, default=False}
 SkirtFrConstraint {type=double, default=0}
]
