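# Physics asset for the WaltzScout medium female body, named for crouching, covering left-side skirt and front pendant secondary motion.
# Back, front and side skirt chains of four capsules each hang from small base anchors jointed to the pelvis, with tight swing and twist limits.
# A three-link pendant chain branches off the first front skirt bone through its own anchor.
# Pelvis and left thigh capsules supply the body shapes; no right-side bodies are present.
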PhysicsAsset /BRCosmetics/Characters/Player/Female/Medium/Bodies/F_MED_WaltzScout/Meshes/F_MED_WaltzScout_PhysicsCrouch
bodies {
 base_pending_fr_l [{shapeType=Capsule, length=0.505, radius=0.505, shapeName=base_pending_fr_l_capsule, rotation=<rot 0 -90 0>}]
 base_skirt_bk_l [{shapeType=Capsule, length=0.505, radius=0.505, shapeName=base_skirt_bk_l_capsule, rotation=<rot 0 -90 0>}]
 base_skirt_fr_l [{shapeType=Capsule, length=0.505, radius=0.505, shapeName=base_skirt_fr_l_capsule, rotation=<rot 0 -90 0>}]
 base_skirt_sd_l [{shapeType=Capsule, length=0.505, radius=0.505, shapeName=base_skirt_sd_l_capsule, rotation=<rot 0 -90 0>}]
 dyn_pending_fr_01_l [{shapeType=Capsule, length=2, radius=1.5, shapeName=dyn_pending_fr_01_l_capsule, center=<2 0 0>, rotation=<rot 90 0 0>}]
 dyn_pending_fr_02_l [{shapeType=Capsule, length=2, radius=1.5, shapeName=dyn_pending_fr_02_l_capsule, center=<2 0 0>, rotation=<rot 90 0 0>}]
 dyn_pending_fr_03_l [{
   shapeType=Capsule
   length=3.5297
   radius=4.8
   shapeName=dyn_pending_fr_03_l_capsule
   center=<3.3578 -1.2276 -1.3383>
   rotation=<rot 81.9652 0 0>
  }]
 dyn_skirt_bk_01_l [{shapeType=Capsule, length=8, radius=3, shapeName=dyn_skirt_bk_01_l_capsule, center=<2 0 1.5748>, rotation=<rot 0 0 90>}]
 dyn_skirt_bk_02_l [{shapeType=Capsule, length=13, radius=5, shapeName=dyn_skirt_bk_02_l_capsule, center=<2 -0.058 5.0669>, rotation=<rot -0.0001 0 82.0392>}]
 dyn_skirt_bk_03_l [{
   shapeType=Capsule
   length=13
   radius=5
   shapeName=dyn_skirt_bk_03_l_capsule
   center=<1.9962 0.0695 4.2506>
   rotation=<rot -0.2137 0.1355 85.8664>
  }]
 dyn_skirt_bk_04_l [{shapeType=Capsule, length=13, radius=5, shapeName=dyn_skirt_bk_04_l_capsule, center=<2 0.038 3.9943>, rotation=<rot 0 0 81.9619>}]
 dyn_skirt_fr_01_l [{shapeType=Capsule, length=8, radius=3, shapeName=dyn_skirt_fr_01_l_capsule, center=<2 -0.0133 1.1177>, rotation=<rot 0.0001 0 97.2079>}]
 dyn_skirt_fr_02_l [{shapeType=Capsule, length=13, radius=2, shapeName=dyn_skirt_fr_02_l_capsule, center=<2 -1.044 3.6785>, rotation=<rot 0.0002 0.0002 101.4455>}]
 dyn_skirt_fr_03_l [{shapeType=Capsule, length=13, radius=4, shapeName=dyn_skirt_fr_03_l_capsule, center=<2 -0.929 5.3432>, rotation=<rot 0.0003 0.0002 103.885>}]
 dyn_skirt_fr_04_l [{
   shapeType=Capsule
   length=13
   radius=5
   shapeName=dyn_skirt_fr_04_l_capsule
   center=<2 -0.5905 5.3376>
   rotation=<rot 0.0003 0.0003 108.7586>
  }]
 dyn_skirt_sd_01_l [{shapeType=Capsule, length=8, radius=3, shapeName=dyn_skirt_sd_01_l_capsule, center=<2.3655 0 2.973>, rotation=<rot 0 0.0015 101.0022>}]
 dyn_skirt_sd_02_l [{shapeType=Capsule, length=18, radius=4, shapeName=dyn_skirt_sd_02_l_capsule, center=<3.4698 0.2252 4.8756>, rotation=<rot 0 0.0016 95.5148>}]
 dyn_skirt_sd_03_l [{shapeType=Capsule, length=20, radius=6, shapeName=dyn_skirt_sd_03_l_capsule, center=<3.0587 0 3.604>, rotation=<rot 0 0.0016 94.8113>}]
 dyn_skirt_sd_04_l [{shapeType=Capsule, length=20, radius=5.5, shapeName=dyn_skirt_sd_04_l_capsule, center=<2.9899 0 4.6519>, rotation=<rot 0 0.0016 96.6341>}]
 pelvis [{shapeType=Capsule, length=11.4233, radius=7.5319, shapeName=pelvis_capsule, center=<0 -1.618 0>}]
 thigh_l [{shapeType=Capsule, length=25.2436, radius=8.1174, shapeName=thigh_l_capsule, center=<-21.1126 0 0.7816>, rotation=<rot 90 0 0>}]
}
constraints [
 {
  bone1Name=dyn_skirt_bk_02_l
  bone2Name=dyn_skirt_bk_01_l
  swing1LimitDegrees=5
  swing1Motion=Locked
  swing2LimitDegrees=10
  twistLimitDegrees=45
  twistMotion=Locked
 }
 {
  bone1Name=dyn_skirt_bk_03_l
  bone2Name=dyn_skirt_bk_02_l
  swing1LimitDegrees=5
  swing1Motion=Locked
  swing2LimitDegrees=15
  twistLimitDegrees=2
 }
 {
  bone1Name=dyn_skirt_bk_04_l
  bone2Name=dyn_skirt_bk_03_l
  swing1LimitDegrees=5
  swing1Motion=Locked
  swing2LimitDegrees=15
  twistLimitDegrees=2
 }
 {
  bone1Name=dyn_skirt_fr_02_l
  bone2Name=dyn_skirt_fr_01_l
  swing1LimitDegrees=7
  swing2LimitDegrees=5
  twistLimitDegrees=7
  twistMotion=Locked
 }
 {
  bone1Name=dyn_skirt_fr_04_l
  bone2Name=dyn_skirt_fr_03_l
  swing1LimitDegrees=5
  swing2LimitDegrees=6
  twistLimitDegrees=7
  twistMotion=Locked
 }
 {
  bone1Name=dyn_skirt_fr_03_l
  bone2Name=dyn_skirt_fr_02_l
  swing1LimitDegrees=3
  swing2LimitDegrees=5
  twistLimitDegrees=7
  twistMotion=Locked
 }
 {bone1Name=dyn_skirt_sd_02_l, bone2Name=dyn_skirt_sd_01_l, swing1LimitDegrees=5, swing2LimitDegrees=5, twistLimitDegrees=5}
 {
  bone1Name=dyn_skirt_sd_03_l
  bone2Name=dyn_skirt_sd_02_l
  swing1LimitDegrees=10
  swing2LimitDegrees=5
  swing2Motion=Locked
  twistLimitDegrees=2
 }
 {
  bone1Name=dyn_skirt_sd_04_l
  bone2Name=dyn_skirt_sd_03_l
  swing1LimitDegrees=10
  swing2LimitDegrees=5
  swing2Motion=Locked
  twistLimitDegrees=2
 }
 {
  bone1Name=dyn_skirt_bk_01_l
  bone2Name=base_skirt_bk_l
  swing1LimitDegrees=5
  swing1Motion=Locked
  swing2LimitDegrees=15
  twistLimitDegrees=45
  twistMotion=Locked
 }
 {
  bone1Name=dyn_skirt_sd_01_l
  bone2Name=base_skirt_sd_l
  swing1LimitDegrees=5
  swing1Motion=Locked
  swing2LimitDegrees=15
  twistLimitDegrees=45
  twistMotion=Locked
 }
 {bone1Name=dyn_skirt_fr_01_l, bone2Name=base_skirt_fr_l, swing1LimitDegrees=5, swing2LimitDegrees=7, twistLimitDegrees=45, twistMotion=Locked}
 {
  bone1Name=thigh_l
  bone2Name=pelvis
  swing1LimitDegrees=45
  swing1Motion=Free
  swing2LimitDegrees=45
  swing2Motion=Free
  twistLimitDegrees=45
  twistMotion=Free
 }
 {
  bone1Name=base_skirt_bk_l
  bone2Name=pelvis
  swing1LimitDegrees=45
  swing1Motion=Free
  swing2LimitDegrees=45
  swing2Motion=Free
  twistLimitDegrees=45
  twistMotion=Free
 }
 {
  bone1Name=base_skirt_fr_l
  bone2Name=pelvis
  swing1LimitDegrees=45
  swing1Motion=Free
  swing2LimitDegrees=45
  swing2Motion=Free
  twistLimitDegrees=45
  twistMotion=Free
 }
 {
  bone1Name=base_skirt_sd_l
  bone2Name=pelvis
  swing1LimitDegrees=45
  swing1Motion=Free
  swing2LimitDegrees=45
  swing2Motion=Free
  twistLimitDegrees=45
  twistMotion=Free
 }
 {
  bone1Name=base_pending_fr_l
  bone2Name=dyn_skirt_fr_01_l
  swing1LimitDegrees=45
  swing1Motion=Free
  swing2LimitDegrees=45
  swing2Motion=Free
  twistLimitDegrees=45
  twistMotion=Free
 }
 {
  bone1Name=dyn_pending_fr_01_l
  bone2Name=base_pending_fr_l
  swing1LimitDegrees=5
  swing2LimitDegrees=20
  twistLimitDegrees=45
  twistMotion=Locked
 }
 {
  bone1Name=dyn_pending_fr_02_l
  bone2Name=dyn_pending_fr_01_l
  swing1LimitDegrees=5
  swing2LimitDegrees=20
  twistLimitDegrees=45
  twistMotion=Locked
 }
 {
  bone1Name=dyn_pending_fr_03_l
  bone2Name=dyn_pending_fr_02_l
  swing1LimitDegrees=10
  swing2LimitDegrees=20
  twistLimitDegrees=45
  twistMotion=Locked
 }
]
