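# Physics asset for the right side of the F_MED_WaltzScout body's skirt, named as a crouch variant.
# Back, front and side chains of four capsule bodies each hang from small base anchors, which join the pelvis with free angular constraints.
# Chain links allow mainly narrow single-axis swing, tightest at the front, with twist mostly locked.
# Pelvis and freely constrained right thigh capsules provide body volumes alongside the skirt.
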PhysicsAsset /BRCosmetics/Characters/Player/Female/Medium/Bodies/F_MED_WaltzScout/Meshes/F_MED_WaltzScout_PhysicsCrouchRight
bodies {
 base_skirt_bk_r [{shapeType=Capsule, length=0.505, radius=0.505, shapeName=base_skirt_bk_r_capsule, rotation=<rot 0 -90 0>}]
 base_skirt_fr_r [{shapeType=Capsule, length=0.505, radius=0.505, shapeName=base_skirt_fr_r_capsule, rotation=<rot 0 -90 0>}]
 base_skirt_sd_r [{shapeType=Capsule, length=0.505, radius=0.505, shapeName=base_skirt_sd_r_capsule, rotation=<rot 0 -90 0>}]
 dyn_skirt_bk_01_r [{shapeType=Capsule, length=8, radius=3, shapeName=dyn_skirt_bk_01_r_capsule, center=<2 0 1.9225>, rotation=<rot 0 0 90>}]
 dyn_skirt_bk_02_r [{shapeType=Capsule, length=13, radius=5, shapeName=dyn_skirt_bk_02_r_capsule, center=<2 0.2156 4.9531>, rotation=<rot -0.0001 0 96.6876>}]
 dyn_skirt_bk_03_r [{shapeType=Capsule, length=13, radius=5, shapeName=dyn_skirt_bk_03_r_capsule, center=<2 0 4.6664>, rotation=<rot -0.0001 0 95.8161>}]
 dyn_skirt_bk_04_r [{shapeType=Capsule, length=13, radius=5, shapeName=dyn_skirt_bk_04_r_capsule, center=<2 0 4.1215>, rotation=<rot -0.0001 0 95.6019>}]
 dyn_skirt_fr_01_r [{shapeType=Capsule, length=8, radius=3, shapeName=dyn_skirt_fr_01_r_capsule, center=<2 0 1.2263>, rotation=<rot 0.0001 -0.0001 76.1692>}]
 dyn_skirt_fr_02_r [{shapeType=Capsule, length=13, radius=5, shapeName=dyn_skirt_fr_02_r_capsule, center=<2 -0.4076 4.1143>, rotation=<rot 0 0 84.342>}]
 dyn_skirt_fr_03_r [{shapeType=Capsule, length=13, radius=5, shapeName=dyn_skirt_fr_03_r_capsule, center=<2 0.2101 3.1854>, rotation=<rot 0.0001 -0.0001 73.2483>}]
 dyn_skirt_fr_04_r [{shapeType=Capsule, length=13, radius=5, shapeName=dyn_skirt_fr_04_r_capsule, center=<2 -0.0041 3.1196>, rotation=<rot 0 0 74.3568>}]
 dyn_skirt_sd_01_r [{shapeType=Capsule, length=8, radius=3, shapeName=dyn_skirt_sd_01_r_capsule, center=<2 0 3.0849>, rotation=<rot 0 0 90>}]
 dyn_skirt_sd_02_r [{shapeType=Capsule, length=13, radius=5, shapeName=dyn_skirt_sd_02_r_capsule, center=<2 0 5.3004>, rotation=<rot 0 0 90>}]
 dyn_skirt_sd_03_r [{shapeType=Capsule, length=13, radius=5, shapeName=dyn_skirt_sd_03_r_capsule, center=<2 0 3.8077>, rotation=<rot 0 0 90>}]
 dyn_skirt_sd_04_r [{shapeType=Capsule, length=13, radius=5, shapeName=dyn_skirt_sd_04_r_capsule, center=<2 0 4.1652>, rotation=<rot 0 0 90>}]
 pelvis [{shapeType=Capsule, length=11.4233, radius=7.5319, shapeName=pelvis_capsule, center=<0 -1.618 0>}]
 thigh_r [{shapeType=Capsule, length=25.2436, radius=8.1174, shapeName=thigh_r_capsule, center=<21.1126 0 -0.7816>, rotation=<rot 90 180 0>}]
}
constraints [
 {
  bone1Name=dyn_skirt_bk_03_r
  bone2Name=dyn_skirt_bk_02_r
  swing1LimitDegrees=5
  swing1Motion=Locked
  swing2LimitDegrees=10
  twistLimitDegrees=2
 }
 {
  bone1Name=dyn_skirt_bk_02_r
  bone2Name=dyn_skirt_bk_01_r
  swing1LimitDegrees=5
  swing1Motion=Locked
  swing2LimitDegrees=10
  twistLimitDegrees=45
  twistMotion=Locked
 }
 {
  bone1Name=dyn_skirt_bk_04_r
  bone2Name=dyn_skirt_bk_03_r
  swing1LimitDegrees=5
  swing1Motion=Locked
  swing2LimitDegrees=10
  twistLimitDegrees=2
 }
 {
  bone1Name=dyn_skirt_fr_02_r
  bone2Name=dyn_skirt_fr_01_r
  swing1LimitDegrees=5
  swing1Motion=Locked
  swing2LimitDegrees=3
  twistLimitDegrees=45
  twistMotion=Locked
 }
 {
  bone1Name=dyn_skirt_fr_03_r
  bone2Name=dyn_skirt_fr_02_r
  swing1LimitDegrees=5
  swing1Motion=Locked
  swing2LimitDegrees=3
  twistLimitDegrees=45
  twistMotion=Locked
 }
 {
  bone1Name=dyn_skirt_fr_04_r
  bone2Name=dyn_skirt_fr_03_r
  swing1LimitDegrees=5
  swing1Motion=Locked
  swing2LimitDegrees=3
  twistLimitDegrees=45
  twistMotion=Locked
 }
 {
  bone1Name=dyn_skirt_sd_02_r
  bone2Name=dyn_skirt_sd_01_r
  swing1LimitDegrees=5
  swing1Motion=Locked
  swing2LimitDegrees=10
  twistLimitDegrees=45
  twistMotion=Locked
 }
 {
  bone1Name=dyn_skirt_sd_03_r
  bone2Name=dyn_skirt_sd_02_r
  swing1LimitDegrees=4
  swing2LimitDegrees=10
  twistLimitDegrees=45
  twistMotion=Locked
 }
 {
  bone1Name=dyn_skirt_sd_04_r
  bone2Name=dyn_skirt_sd_03_r
  swing1LimitDegrees=4
  swing2LimitDegrees=10
  twistLimitDegrees=45
  twistMotion=Locked
 }
 {
  bone1Name=dyn_skirt_bk_01_r
  bone2Name=base_skirt_bk_r
  swing1LimitDegrees=5
  swing1Motion=Locked
  swing2LimitDegrees=15
  twistLimitDegrees=45
  twistMotion=Locked
 }
 {
  bone1Name=dyn_skirt_sd_01_r
  bone2Name=base_skirt_sd_r
  swing1LimitDegrees=5
  swing1Motion=Locked
  swing2LimitDegrees=15
  twistLimitDegrees=45
  twistMotion=Locked
 }
 {
  bone1Name=dyn_skirt_fr_01_r
  bone2Name=base_skirt_fr_r
  swing1LimitDegrees=5
  swing1Motion=Locked
  swing2LimitDegrees=15
  twistLimitDegrees=45
  twistMotion=Locked
 }
 {
  bone1Name=thigh_r
  bone2Name=pelvis
  swing1LimitDegrees=45
  swing1Motion=Free
  swing2LimitDegrees=45
  swing2Motion=Free
  twistLimitDegrees=45
  twistMotion=Free
 }
 {
  bone1Name=base_skirt_bk_r
  bone2Name=pelvis
  swing1LimitDegrees=45
  swing1Motion=Free
  swing2LimitDegrees=45
  swing2Motion=Free
  twistLimitDegrees=45
  twistMotion=Free
 }
 {
  bone1Name=base_skirt_fr_r
  bone2Name=pelvis
  swing1LimitDegrees=45
  swing1Motion=Free
  swing2LimitDegrees=45
  swing2Motion=Free
  twistLimitDegrees=45
  twistMotion=Free
 }
 {
  bone1Name=base_skirt_sd_r
  bone2Name=pelvis
  swing1LimitDegrees=45
  swing1Motion=Free
  swing2LimitDegrees=45
  swing2Motion=Free
  twistLimitDegrees=45
  twistMotion=Free
 }
]
