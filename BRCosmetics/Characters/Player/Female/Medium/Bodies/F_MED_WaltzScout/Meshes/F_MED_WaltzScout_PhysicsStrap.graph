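# Physics asset for dynamic straps on the WaltzScout female medium body, built on the pelvis, left thigh and strap bones.
# Front and back four-bone capsule chains hang from base bones pinned to the pelvis, with tight swing and twist limits between links.
# Both chain ends join dyn_strap_03, which links through base_strap_l to the pelvis by free-rotation constraints, closing the loops.
# Capsules on the pelvis and left thigh supply body shapes alongside the straps.
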PhysicsAsset /BRCosmetics/Characters/Player/Female/Medium/Bodies/F_MED_WaltzScout/Meshes/F_MED_WaltzScout_PhysicsStrap
bodies {
 base_strap_l [{shapeType=Capsule, length=0.505, radius=0.505, shapeName=base_strap_l_capsule, rotation=<rot 0 -90 0>}]
 base_strap_part_bk [{shapeType=Capsule, length=0.505, radius=0.505, shapeName=base_strap_part_bk_capsule, rotation=<rot 0 -90 0>}]
 base_strap_part_fr [{shapeType=Capsule, length=0.505, radius=0.505, shapeName=base_strap_part_fr_capsule, rotation=<rot 0 -90 0>}]
 dyn_strap_03 [{shapeType=Capsule, length=0.505, radius=0.505, shapeName=dyn_strap_03_capsule, rotation=<rot 0 -90 0>}]
 dyn_strap_part_bk_01 [{shapeType=Capsule, length=2.5, radius=1, shapeName=dyn_strap_part_bk_01_capsule, center=<2 0 0>, rotation=<rot 90 0 0>}]
 dyn_strap_part_bk_02 [{shapeType=Capsule, length=2, radius=4, shapeName=dyn_strap_part_bk_02_capsule, center=<2 0.6119 0>, rotation=<rot 78.2083 -0.0001 -0.0001>}]
 dyn_strap_part_bk_03 [{
   shapeType=Capsule
   length=2
   radius=3.3
   shapeName=dyn_strap_part_bk_03_capsule
   center=<1.8792 0.9258 -0.6236>
   rotation=<rot 79.033 -0.0001 -0.0001>
  }]
 dyn_strap_part_bk_04 [{
   shapeType=Capsule
   length=5
   radius=3.3
   shapeName=dyn_strap_part_bk_04_capsule
   center=<0.4644 0.9258 -0.2015>
   rotation=<rot 72.4802 -180 -179.9999>
  }]
 dyn_strap_part_fr_01 [{shapeType=Capsule, length=2.5, radius=1, shapeName=dyn_strap_part_fr_01_capsule, center=<2 0 0>, rotation=<rot 90 0 0>}]
 dyn_strap_part_fr_02 [{shapeType=Capsule, length=3, radius=3.5, shapeName=dyn_strap_part_fr_02_capsule, center=<0.885 0.6232 0>, rotation=<rot 90 0 0>}]
 dyn_strap_part_fr_03 [{shapeType=Capsule, length=2, radius=2.5, shapeName=dyn_strap_part_fr_03_capsule, center=<2 -0.0057 0>, rotation=<rot 90 0 0>}]
 dyn_strap_part_fr_04 [{shapeType=Capsule, length=5, radius=2.5, shapeName=dyn_strap_part_fr_04_capsule, center=<2 0.7194 0>, rotation=<rot 90 12.9219 0>}]
 pelvis [{shapeType=Capsule, length=11.4233, radius=7.5319, shapeName=pelvis_capsule, center=<0 -1.618 0>}]
 thigh_l [
  {shapeType=Capsule, length=30, radius=8.1174, shapeName=thigh_l_capsule, center=<-19.8562 0 0.7816>, rotation=<rot 90 0 0>}
  {shapeType=Capsule, length=9.3974, radius=3.6026, shapeName=thigh_l_capsule, center=<-0.8342 -1.7159 0.7637>, rotation=<rot 90 0 0>}
 ]
}
constraints [
 {
  bone1Name=thigh_l
  bone2Name=pelvis
  swing1LimitDegrees=45
  swing1Motion=Free
  swing2LimitDegrees=45
  swing2Motion=Free
  twistLimitDegrees=45
  twistMotion=Free
 }
 {bone1Name=dyn_strap_part_fr_04, bone2Name=dyn_strap_part_fr_03, swing1LimitDegrees=8, swing2LimitDegrees=3, twistLimitDegrees=5}
 {bone1Name=dyn_strap_part_fr_03, bone2Name=dyn_strap_part_fr_02, swing1LimitDegrees=10, swing2LimitDegrees=3, twistLimitDegrees=5}
 {bone1Name=dyn_strap_part_fr_02, bone2Name=dyn_strap_part_fr_01, swing1LimitDegrees=12, swing2LimitDegrees=3, twistLimitDegrees=5}
 {bone1Name=dyn_strap_part_fr_01, bone2Name=base_strap_part_fr, swing1LimitDegrees=7, swing2LimitDegrees=3, twistLimitDegrees=5}
 {bone1Name=dyn_strap_part_bk_04, bone2Name=dyn_strap_part_bk_03, swing1LimitDegrees=4, swing2LimitDegrees=3, twistLimitDegrees=5}
 {bone1Name=dyn_strap_part_bk_03, bone2Name=dyn_strap_part_bk_02, swing1LimitDegrees=8, swing2LimitDegrees=3, twistLimitDegrees=5}
 {bone1Name=dyn_strap_part_bk_02, bone2Name=dyn_strap_part_bk_01, swing1LimitDegrees=16, swing2LimitDegrees=3, twistLimitDegrees=5}
 {
  bone1Name=base_strap_part_fr
  bone2Name=pelvis
  swing1LimitDegrees=45
  swing1Motion=Free
  swing2LimitDegrees=45
  swing2Motion=Free
  twistLimitDegrees=45
  twistMotion=Free
 }
 {
  bone1Name=base_strap_part_bk
  bone2Name=pelvis
  swing1LimitDegrees=45
  swing1Motion=Free
  swing2LimitDegrees=45
  swing2Motion=Free
  twistLimitDegrees=45
  twistMotion=Free
 }
 {bone1Name=dyn_strap_part_bk_01, bone2Name=base_strap_part_bk, swing1LimitDegrees=10, swing2LimitDegrees=3, twistLimitDegrees=5}
 {
  bone1Name=dyn_strap_03
  bone2Name=base_strap_l
  swing1LimitDegrees=45
  swing1Motion=Free
  swing2LimitDegrees=45
  swing2Motion=Free
  twistLimitDegrees=45
  twistMotion=Free
 }
 {
  bone1Name=base_strap_l
  bone2Name=pelvis
  swing1LimitDegrees=45
  swing1Motion=Free
  swing2LimitDegrees=45
  swing2Motion=Free
  twistLimitDegrees=45
  twistMotion=Free
 }
 {
  bone1Name=dyn_strap_part_fr_04
  bone2Name=dyn_strap_03
  swing1LimitDegrees=45
  swing1Motion=Free
  swing2LimitDegrees=45
  swing2Motion=Free
  twistLimitDegrees=45
  twistMotion=Free
 }
 {
  bone1Name=dyn_strap_part_bk_04
  bone2Name=dyn_strap_03
  swing1LimitDegrees=45
  swing1Motion=Free
  swing2LimitDegrees=45
  swing2Motion=Free
  twistLimitDegrees=45
  twistMotion=Free
 }
]
