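# Post-process style AnimBP for the female WaltzScout cosmetic body, animating its skirt, straps, bows and lollipop accessory.
# Copies the attached parent's pose through a state-driven Control Rig, then caches it.
# Standing and crouch branches add separate RigidBody setups, AnimDynamics and bone corrections, blended on bIsCrouching.
# Shared skirt corrections and bow/foot-strap dynamics follow; disabling bRigidBodyEnabled outputs the plain Control Rig pose.
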AnimBlueprint /BRCosmetics/Characters/Player/Female/Medium/Bodies/F_MED_WaltzScout/Meshes/F_MED_WaltzScout_AnimBP
target_skeleton=&/F_MED_WaltzScout_Skeleton
graphs [
 AnimGraph {nodes [
   n1 = Use cached pose 'CrouchCache'()
   n2 = Use cached pose 'StandingCache'()
   n3 = Get bIsCrouching()
   n4 = "Transform (Modify) Bone - Bone: base_skirt_fr_r"!(
    ComponentPose=Local To Component(LocalPose=Blend Poses by bool(BlendPose_0=@n1, BlendPose_1=@n2, BlendTime_0=0.5, bActiveValue=@n3))
    bAlphaBoolEnabled:bool=True
    Rotation="-9.615,0.243,3.777"
   ) -> (Pose) {
    alphaInputType=Bool
    rotationMode=BMM_Additive
    rotationSpace=BCS_BoneSpace
    scaleMode=BMM_Additive
    scaleSpace=BCS_BoneSpace
    translationMode=BMM_Additive
    translationSpace=BCS_BoneSpace
    alphaBoolBlend {blendInTime=1, blendOutTime=0.2}
    alphaScaleBiasClamp {bInterpResult=true}
    boneToModify {boneName=base_skirt_fr_r}
    rotation=<rot -9.615 0.243 3.777>
   }
   n5 = "Transform (Modify) Bone - Bone: base_skirt_fr_c"!(ComponentPose=@n4, bAlphaBoolEnabled:bool=True, Rotation="-8.186,6.398,-9.133") -> (Pose) {
    alphaInputType=Bool
    rotationMode=BMM_Additive
    rotationSpace=BCS_BoneSpace
    scaleMode=BMM_Additive
    scaleSpace=BCS_BoneSpace
    translationMode=BMM_Additive
    translationSpace=BCS_BoneSpace
    alphaBoolBlend {blendInTime=1, blendOutTime=0.2}
    alphaScaleBiasClamp {bInterpResult=true}
    boneToModify {boneName=base_skirt_fr_c}
    rotation=<rot -8.186 6.398 -9.133>
    translation=<2.3403 0.3819 -0.257>
   }
   n6 = "Transform (Modify) Bone - Bone: base_skirt_fr_l"!(ComponentPose=@n5, bAlphaBoolEnabled:bool=True, Rotation="-3.606,1.020,-15.002") -> (Pose) {
    alphaInputType=Bool
    rotationMode=BMM_Additive
    rotationSpace=BCS_BoneSpace
    scaleMode=BMM_Additive
    scaleSpace=BCS_BoneSpace
    translationMode=BMM_Additive
    translationSpace=BCS_BoneSpace
    alphaBoolBlend {blendInTime=0.5, blendOutTime=0.2}
    alphaScaleBiasClamp {bInterpResult=true}
    boneToModify {boneName=base_skirt_fr_l}
    rotation=<rot -3.606 1.02 -15.002>
   }
   n7 = Anim Dynamics - Bone: dyn_bowUp_01_r!(ComponentPose=@n6, Alpha=1, GravityOverride:Vector="0,0,-980", ExternalForce="0,0,0") -> (Pose) {
    angularDampingOverride=1
    angularSpringConstant=20000
    bAngularSpring=true
    bOverrideAngularDamping=true
    bUseGravityOverride=true
    gravityScale=0.35
    simulationSpace=RootRelative
    boundBone {boneName=dyn_bowUp_01_r}
    componentAppliedLinearAccClamp=<3000 3000 3000>
    componentLinearAccScale=<1 1 1>
    componentLinearVelScale=<1 1 1>
    gravityOverride {z=-980}
    physicsBodyDefinitions [{
      boundBone {boneName=dyn_bowUp_01_r}
      boxExtents=<20 30 30>
      constraintSetup {angularLimitsMax=<2 0 2>, angularLimitsMin=<-2 -15 -2>, angularTarget=<1 0 0>}
      localJointOffset=<4 0 0>
     }]
    relativeSpaceBone {boneName=spine_03}
   }
   n8 = Anim Dynamics - Bone: dyn_bowUp_01_l!(ComponentPose=@n7, Alpha=1, GravityOverride:Vector="0,0,-980", ExternalForce="0,0,0") -> (Pose) {
    angularDampingOverride=1
    angularSpringConstant=20000
    bAngularSpring=true
    bOverrideAngularDamping=true
    bUseGravityOverride=true
    gravityScale=0.35
    simulationSpace=RootRelative
    boundBone {boneName=dyn_bowUp_01_l}
    componentAppliedLinearAccClamp=<3000 3000 3000>
    componentLinearAccScale=<1 1 1>
    componentLinearVelScale=<1 1 1>
    gravityOverride {z=-980}
    physicsBodyDefinitions [{
      boundBone {boneName=dyn_bowUp_01_l}
      boxExtents=<20 30 30>
      constraintSetup {angularLimitsMax=<0 15 5>, angularLimitsMin=<0 0 -5>, angularTarget=<1 0 0>}
      localJointOffset=<4 0 0>
     }]
    relativeSpaceBone {boneName=spine_03}
   }
   n9 = Anim Dynamics - Bone: dyn_bow_01_l!(ComponentPose=@n8, Alpha=1, GravityOverride:Vector="0,0,-980", ExternalForce="0,0,0") -> (Pose) {
    angularDampingOverride=1
    angularSpringConstant=500
    bAngularSpring=true
    bOverrideAngularDamping=true
    bUseGravityOverride=true
    gravityScale=0.35
    simulationSpace=RootRelative
    boundBone {boneName=dyn_bow_01_l}
    componentAppliedLinearAccClamp=<3000 3000 3000>
    componentLinearAccScale=<0.6 0.6 0.6>
    componentLinearVelScale=<1 1 1>
    gravityOverride {z=-980}
    physicsBodyDefinitions [{
      boundBone {boneName=dyn_bow_01_l}
      boxExtents=<27 20 15>
      constraintSetup {angularLimitsMax=<5 0 10>, angularLimitsMin=<-5 -40 -5>, angularTarget=<1 0 0>}
      localJointOffset=<7 0 0>
     }]
    relativeSpaceBone {boneName=spine_01}
   }
   n10 = Anim Dynamics - Bone: dyn_bow_02_l!(ComponentPose=@n9, Alpha=1, GravityOverride:Vector="0,0,-980", ExternalForce="0,0,0") -> (Pose) {
    angularDampingOverride=1
    angularSpringConstant=500
    bAngularSpring=true
    bOverrideAngularDamping=true
    bUseGravityOverride=true
    gravityScale=0.35
    simulationSpace=RootRelative
    boundBone {boneName=dyn_bow_02_l}
    componentAppliedLinearAccClamp=<3000 3000 3000>
    componentLinearAccScale=<0.6 0.6 0.6>
    componentLinearVelScale=<1 1 1>
    gravityOverride {z=-980}
    physicsBodyDefinitions [{
      boundBone {boneName=dyn_bow_02_l}
      boxExtents=<27 20 15>
      constraintSetup {angularLimitsMax=<5 0 10>, angularLimitsMin=<-5 -40 -10>, angularTarget=<1 0 0>}
      localJointOffset=<7 0 0>
     }]
    relativeSpaceBone {boneName=spine_01}
   }
   n11 = Anim Dynamics - Bone: dyn_bow_01_r!(ComponentPose=@n10, Alpha=1, GravityOverride:Vector="0,0,-980", ExternalForce="0,0,0") -> (Pose) {
    angularDampingOverride=1
    angularSpringConstant=500
    bAngularSpring=true
    bOverrideAngularDamping=true
    bUseGravityOverride=true
    gravityScale=0.35
    simulationSpace=RootRelative
    boundBone {boneName=dyn_bow_01_r}
    componentAppliedLinearAccClamp=<3000 3000 3000>
    componentLinearAccScale=<0.6 0.6 0.6>
    componentLinearVelScale=<1 1 1>
    gravityOverride {z=-980}
    physicsBodyDefinitions [{
      boundBone {boneName=dyn_bow_01_r}
      boxExtents=<27 20 15>
      constraintSetup {angularLimitsMax=<5 0 10>, angularLimitsMin=<-5 -40 -10>, angularTarget=<1 0 0>}
      localJointOffset=<-7 0 0>
     }]
    relativeSpaceBone {boneName=spine_01}
   }
   n12 = Anim Dynamics - Bone: dyn_bow_02_r!(ComponentPose=@n11, Alpha=1, GravityOverride:Vector="0,0,-980", ExternalForce="0,0,0") -> (Pose) {
    angularDampingOverride=1
    angularSpringConstant=500
    bAngularSpring=true
    bOverrideAngularDamping=true
    bUseGravityOverride=true
    gravityScale=0.35
    simulationSpace=RootRelative
    boundBone {boneName=dyn_bow_02_r}
    componentAppliedLinearAccClamp=<3000 3000 3000>
    componentLinearAccScale=<0.6 0.6 0.6>
    componentLinearVelScale=<1 1 1>
    gravityOverride {z=-980}
    physicsBodyDefinitions [{
      boundBone {boneName=dyn_bow_02_r}
      boxExtents=<27 20 15>
      constraintSetup {angularLimitsMax=<5 0 10>, angularLimitsMin=<-5 -40 -10>, angularTarget=<1 0 0>}
      localJointOffset=<-7 0 0>
     }]
    relativeSpaceBone {boneName=spine_01}
   }
   n13 = Anim Dynamics - Bone: dyn_footStrap_in_r!(ComponentPose=@n12, Alpha=1, GravityOverride:Vector="0,0,-980", ExternalForce="0,0,0") -> (Pose) {
    angularDampingOverride=1
    angularSpringConstant=20000
    bAngularSpring=true
    bOverrideAngularDamping=true
    bUseGravityOverride=true
    gravityScale=0.35
    simulationSpace=BoneRelative
    boundBone {boneName=dyn_footStrap_in_r}
    componentAppliedLinearAccClamp=<3000 3000 3000>
    componentLinearAccScale=<1 1 1>
    componentLinearVelScale=<1 1 1>
    gravityOverride {z=-980}
    physicsBodyDefinitions [{
      boundBone {boneName=dyn_footStrap_in_r}
      boxExtents=<20 20 20>
      constraintSetup {angularLimitsMax=<0 0 5>, angularLimitsMin=<0 -80 0>, angularTarget=<1 0 0>}
      localJointOffset=<4 0 0>
     }]
    relativeSpaceBone {boneName=pelvis}
   }
   n14 = Anim Dynamics - Bone: dyn_footStrap_out_r!(ComponentPose=@n13, Alpha=1, GravityOverride:Vector="0,0,-980", ExternalForce="0,0,0") -> (Pose) {
    angularDampingOverride=1
    angularSpringConstant=20000
    bAngularSpring=true
    bOverrideAngularDamping=true
    bUseGravityOverride=true
    gravityScale=0.35
    simulationSpace=BoneRelative
    boundBone {boneName=dyn_footStrap_out_r}
    componentAppliedLinearAccClamp=<3000 3000 3000>
    componentLinearAccScale=<1 1 1>
    componentLinearVelScale=<1 1 1>
    gravityOverride {z=-980}
    physicsBodyDefinitions [{
      boundBone {boneName=dyn_footStrap_out_r}
      boxExtents=<20 20 20>
      constraintSetup {angularLimitsMax=<0 0 5>, angularLimitsMin=<0 -80 0>, angularTarget=<1 0 0>}
      localJointOffset=<4 0 0>
     }]
    relativeSpaceBone {boneName=pelvis}
   }
   n15 = Anim Dynamics - Bone: dyn_footStrap_out_l!(ComponentPose=@n14, Alpha=1, GravityOverride:Vector="0,0,-980", ExternalForce="0,0,0") -> (Pose) {
    angularDampingOverride=1
    angularSpringConstant=20000
    bAngularSpring=true
    bOverrideAngularDamping=true
    bUseGravityOverride=true
    gravityScale=0.35
    simulationSpace=BoneRelative
    boundBone {boneName=dyn_footStrap_out_l}
    componentAppliedLinearAccClamp=<3000 3000 3000>
    componentLinearAccScale=<1 1 1>
    componentLinearVelScale=<1 1 1>
    gravityOverride {z=-980}
    physicsBodyDefinitions [{
      boundBone {boneName=dyn_footStrap_out_l}
      boxExtents=<20 20 20>
      constraintSetup {angularLimitsMax=<0 0 5>, angularLimitsMin=<0 -80 0>, angularTarget=<1 0 0>}
      localJointOffset=<4 0 0>
     }]
    relativeSpaceBone {boneName=pelvis}
   }
   n16 = Anim Dynamics - Bone: dyn_footStrap_in_l!(ComponentPose=@n15, Alpha=1, GravityOverride:Vector="0,0,-980", ExternalForce="0,0,0") -> (Pose) {
    angularDampingOverride=1
    angularSpringConstant=20000
    bAngularSpring=true
    bOverrideAngularDamping=true
    bUseGravityOverride=true
    gravityScale=0.35
    simulationSpace=BoneRelative
    boundBone {boneName=dyn_footStrap_in_l}
    componentAppliedLinearAccClamp=<3000 3000 3000>
    componentLinearAccScale=<1 1 1>
    componentLinearVelScale=<1 1 1>
    gravityOverride {z=-980}
    physicsBodyDefinitions [{
      boundBone {boneName=dyn_footStrap_in_l}
      boxExtents=<20 20 20>
      constraintSetup {angularLimitsMax=<0 0 5>, angularLimitsMin=<0 -80 0>, angularTarget=<1 0 0>}
      localJointOffset=<4 0 0>
     }]
    relativeSpaceBone {boneName=pelvis}
   }
   n17 = "Transform (Modify) Bone - Bone: dyn_skirt_fr_04_c"!(ComponentPose=@n16, bAlphaBoolEnabled:bool=True, Translation="8.641,0.000,4.155", Rotation="0.014,-0.011,-0.815", Scale="1,1,1") -> (Pose) {
    alphaInputType=Bool
    rotationMode=BMM_Additive
    rotationSpace=BCS_BoneSpace
    scaleMode=BMM_Additive
    scaleSpace=BCS_BoneSpace
    translationMode=BMM_Additive
    translationSpace=BCS_BoneSpace
    alphaBoolBlend {blendInTime=0.1, blendOutTime=0.2}
    alphaScaleBiasClamp {bInterpResult=true}
    boneToModify {boneName=dyn_skirt_fr_04_c}
    rotation=<rot 0.014 -0.011 -0.815>
    translation {x=8.641, z=4.155}
   }
   n18 = "Transform (Modify) Bone - Bone: dyn_skirt_bk_03_c"!(ComponentPose=@n17, bAlphaBoolEnabled:bool=True, Translation="0,0,0", Rotation="-15.355,-0.016,0.008", Scale="1,1,1") -> (Pose) {
    alphaInputType=Bool
    rotationMode=BMM_Additive
    rotationSpace=BCS_BoneSpace
    scaleMode=BMM_Additive
    scaleSpace=BCS_BoneSpace
    translationMode=BMM_Additive
    translationSpace=BCS_BoneSpace
    alphaBoolBlend {blendInTime=0.1, blendOutTime=0.2}
    alphaScaleBiasClamp {bInterpResult=true}
    boneToModify {boneName=dyn_skirt_bk_03_c}
    rotation=<rot -15.355 -0.0161 0.008>
   }
   n19 = "Transform (Modify) Bone - Bone: dyn_skirt_bk_03_l"!(ComponentPose=@n18, bAlphaBoolEnabled:bool=True, Translation="0,0,0", Rotation="-21.012,9.816,-3.550", Scale="1,1,1") -> (Pose) {
    alphaInputType=Bool
    rotationMode=BMM_Additive
    rotationSpace=BCS_BoneSpace
    scaleMode=BMM_Additive
    scaleSpace=BCS_BoneSpace
    translationMode=BMM_Additive
    translationSpace=BCS_BoneSpace
    alphaBoolBlend {blendInTime=0.1, blendOutTime=0.2}
    alphaScaleBiasClamp {bInterpResult=true}
    boneToModify {boneName=dyn_skirt_bk_03_l}
    rotation=<rot -21.012 9.816 -3.55>
   }
   n20 = "Transform (Modify) Bone - Bone: dyn_skirt_bk_03_r"!(ComponentPose=@n19, bAlphaBoolEnabled:bool=True, Translation="0,0,0", Rotation="-14.356,-5.023,1.246", Scale="1,1,1") -> (Pose) {
    alphaInputType=Bool
    rotationMode=BMM_Additive
    rotationSpace=BCS_BoneSpace
    scaleMode=BMM_Additive
    scaleSpace=BCS_BoneSpace
    translationMode=BMM_Additive
    translationSpace=BCS_BoneSpace
    alphaBoolBlend {blendInTime=0.1, blendOutTime=0.2}
    alphaScaleBiasClamp {bInterpResult=true}
    boneToModify {boneName=dyn_skirt_bk_03_r}
    rotation=<rot -14.356 -5.023 1.246>
   }
   n21 = "Transform (Modify) Bone - Bone: base_skirt_bk_r"!(ComponentPose=@n20, bAlphaBoolEnabled:bool=True, Translation="-0.055,0.010,0.562", Rotation="5.638,-10.912,0.000", Scale="1,1,1") -> (Pose) {
    alphaInputType=Bool
    rotationMode=BMM_Additive
    rotationSpace=BCS_BoneSpace
    scaleMode=BMM_Additive
    scaleSpace=BCS_BoneSpace
    translationMode=BMM_Additive
    translationSpace=BCS_BoneSpace
    alphaBoolBlend {blendInTime=0.1, blendOutTime=0.2}
    alphaScaleBiasClamp {bInterpResult=true}
    boneToModify {boneName=base_skirt_bk_r}
    rotation {pitch=5.638, yaw=-10.912}
    translation=<-0.055 0.01 0.562>
   }
   n22 = "Transform (Modify) Bone - Bone: base_skirt_bk_c"!(ComponentPose=@n21, bAlphaBoolEnabled:bool=True, Translation="-0.000,-0.205,1.233", Rotation="0.618,-0.022,0.011", Scale="1,1,1") -> (Pose) {
    alphaInputType=Bool
    rotationMode=BMM_Additive
    rotationSpace=BCS_BoneSpace
    scaleMode=BMM_Additive
    scaleSpace=BCS_BoneSpace
    translationMode=BMM_Additive
    translationSpace=BCS_BoneSpace
    alphaBoolBlend {blendInTime=0.1, blendOutTime=0.2}
    alphaScaleBiasClamp {bInterpResult=true}
    boneToModify {boneName=base_skirt_bk_c}
    rotation=<rot 0.618 -0.022 0.011>
    translation {y=-0.205, z=1.233}
   }
   n23 = "Transform (Modify) Bone - Bone: base_skirt_fr_c"!(ComponentPose=@n22, Alpha:float=1, Translation:Vector="13.810,-0.006,-0.172", Rotation="0.285,-0.136,-18.011", Scale:Vector="1,1,1") -> (Pose) {
    rotationMode=BMM_Additive
    rotationSpace=BCS_BoneSpace
    scaleMode=BMM_Additive
    scaleSpace=BCS_BoneSpace
    translationMode=BMM_Additive
    translationSpace=BCS_BoneSpace
    alphaBoolBlend {blendInTime=0.1, blendOutTime=0.2}
    alphaScaleBiasClamp {bInterpResult=true}
    boneToModify {boneName=base_skirt_fr_c}
    rotation=<rot 0.285 -0.136 -18.011>
    translation=<13.81 -0.006 -0.172>
   }
   Output Pose(Result=Blend Poses by bool(BlendPose_0=Component To Local(ComponentPose=@n23), BlendPose_1=Use cached pose 'ControlRigCache'(), bActiveValue=Get bRigidBodyEnabled())) {layerGroup=None}
   n24 = Control Rig!(
    Source=Copy Pose From Mesh() {bUseAttachedParent=true}
    Alpha
    CrouchingBool:bool=Get bIsCrouching()
    DBNOBool:bool=Get bIsDBNO()
    FallingOnGround:double=Get Alpha_FallingOnGround() -> (Alpha_FallingOnGround:double)
    GrindRail:bool=Get bIsOnGrindRail()
    DBNOCarried:bool=Get bIsDBNOCarried()
    Idle:double=Get Idle() -> (Idle:double)
    Jumping:bool=Get bIsJumping()
    HangGlider:double=Get HangGlidingAlpha() -> (HangGlidingAlpha:double)
    SkirtFrConstraint:double=Get Alpha_SkirtFRConstraint() -> (Alpha_SkirtFRConstraint:double)
    Pistol:double=Get Alpha_Pistol() -> (Alpha_Pistol:double)
    PlayingEasternBloc:bool=Get BoolPlayingEasternBloc() -> (BoolPlayingEasternBloc:bool)
   ) -> (Pose) {controlRigAssetReference {blueprintRigClass=&/F_MED_WaltzScout_CtrlRig.F_MED_WaltzScout_CtrlRig_C}}
   Save cached pose 'ControlRigCache'(Pose=@n24) {cacheName=ControlRigCache}
   n25 = "Constraint - base_pending_fr_l (R)"(ComponentPose=Local To Component(LocalPose=Use cached pose 'ControlRigCache'()), Alpha=0.7, ConstraintWeights_0=1) {
    boneToModify {boneName=base_pending_fr_l}
    constraintSetup [{transformType=Rotation, targetBone {boneName=dyn_skirt_fr_02_c}}]
    constraintWeights [1]
   }
   n26 = Anim Dynamics - Bone: base_strap_l!(ComponentPose=@n25, bAlphaBoolEnabled:bool=True, GravityOverride:Vector="0,0,-980", ExternalForce="0,0,0") -> (Pose) {
    alphaInputType=Bool
    angularDampingOverride=0.9
    angularSpringConstant=15000
    bAngularSpring=true
    bOverrideAngularDamping=true
    bUseGravityOverride=true
    gravityScale=0.35
    simulationSpace=RootRelative
    alphaBoolBlend {blendInTime=0.2, blendOutTime=0.2}
    boundBone {boneName=base_strap_l}
    componentAppliedLinearAccClamp=<2500 2500 2500>
    componentLinearAccScale=<1 1 1>
    componentLinearVelScale=<1 1 1>
    gravityOverride {z=-980}
    physicsBodyDefinitions [{
      boundBone {boneName=base_strap_l}
      boxExtents=<30 30 30>
      constraintSetup {angularLimitsMax=<0 0 5>, angularLimitsMin=<-20 -40 -5>, angularTarget=<1 0 0>}
      localJointOffset=<15 0 0>
     }]
    simSpaceSettings {simSpaceAngularAlpha=0.5}
   }
   n27 = Anim Dynamics - Bone: base_strap_l!(ComponentPose=@n26, Alpha:float=1, GravityOverride="0,0,-980", ExternalForce="0,0,0") -> (Pose) {
    angularDampingOverride=0.9
    angularSpringConstant=15000
    bAngularSpring=true
    bOverrideAngularDamping=true
    bUseGravityOverride=true
    gravityScale=0.35
    simulationSpace=RootRelative
    boundBone {boneName=base_strap_l}
    componentAppliedLinearAccClamp=<2500 2500 2500>
    componentLinearAccScale=<0.7 0.7 0.7>
    componentLinearVelScale=<0.7 0.7 0.7>
    gravityOverride {z=-980}
    physicsBodyDefinitions [{
      boundBone {boneName=base_strap_l}
      boxExtents=<30 30 30>
      constraintSetup {angularLimitsMax=<0 0 5>, angularLimitsMin=<-20 -40 -5>, angularTarget=<1 0 0>}
      localJointOffset=<6 0 0>
     }]
   }
   n28 = RigidBody!(
    ComponentPose=@n27
    Alpha
    OverrideWorldGravity:Vector="0,0,0"
    ExternalForce
    SimSpaceSettings:Sim Space Settings Structure=(WorldAlpha=0,VelocityScaleZ=1,DampingAlpha=1,MaxLinearVelocity=10000,MaxAngularVelocity=10000,MaxLinearAcceleration=10000,MaxAngularAcceleration=10000,ExternalLinearDragV=(X=0,Y=0,Z=0),ExternalLinearVelocity=(X=0,Y=0,Z=0),ExternalAngularVelocity=(X=0,Y=0,Z=0))
   ) -> (Pose) {
    bOverrideWorldGravity=true
    overridePhysicsAsset=&/F_MED_WaltzScout_Physics
    simulationSpace=BaseBoneSpace
    baseBoneRef {boneName=thigh_l}
    componentAppliedLinearAccClamp=<3000 3000 3000>
   }
   n29 = RigidBody(
    ComponentPose=@n28
    OverrideWorldGravity="0,0,0"
    SimSpaceSettings=(WorldAlpha=0,VelocityScaleZ=1,DampingAlpha=1,MaxLinearVelocity=10000,MaxAngularVelocity=10000,MaxLinearAcceleration=10000,MaxAngularAcceleration=10000,ExternalLinearDragV=(X=0,Y=0,Z=0),ExternalLinearVelocity=(X=0,Y=0,Z=0),ExternalAngularVelocity=(X=0,Y=0,Z=0))
   ) {
    bOverrideWorldGravity=true
    overridePhysicsAsset=&/F_MED_WaltzScout_PhysicsRight
    simulationSpace=BaseBoneSpace
    baseBoneRef {boneName=thigh_r}
    componentAppliedLinearAccClamp=<3000 3000 3000>
   }
   n30 = RigidBody(
    ComponentPose=@n29
    OverrideWorldGravity="0,0,0"
    SimSpaceSettings=(WorldAlpha=1,VelocityScaleZ=1,DampingAlpha=1,MaxLinearVelocity=300,MaxAngularVelocity=300,MaxLinearAcceleration=2000,MaxAngularAcceleration=200,ExternalLinearDragV=(X=2,Y=2,Z=2),ExternalLinearVelocity=(X=0,Y=0,Z=0),ExternalAngularVelocity=(X=0,Y=0,Z=0))
   ) {
    bOverrideWorldGravity=true
    overridePhysicsAsset=&/F_MED_WaltzScout_PhysicsStrap
    simulationSpace=BaseBoneSpace
    alphaScaleBiasClamp {bInterpResult=true}
    baseBoneRef {boneName=pelvis}
    componentAppliedLinearAccClamp=<3000 3000 3000>
    simSpaceSettings {
     maxAngularAcceleration=200
     maxAngularVelocity=300
     maxLinearAcceleration=2000
     maxLinearVelocity=300
     worldAlpha=1
     externalLinearDragV=<2 2 2>
    }
   }
   n31 = Anim Dynamics - Bone: dyn_strap_part_fr_01!(ComponentPose=@n30, Alpha=1, GravityOverride:Vector="0,0,-980", ExternalForce="0,0,0") -> (Pose) {
    angularDampingOverride=0.999
    angularSpringConstant=15000
    bAngularSpring=true
    bOverrideAngularDamping=true
    bUseGravityOverride=true
    gravityScale=0.35
    alphaScaleBiasClamp {bInterpResult=true}
    boundBone {boneName=dyn_strap_part_fr_01}
    componentAppliedLinearAccClamp=<3000 3000 3000>
    componentLinearAccScale=<1 1 1>
    componentLinearVelScale=<1 1 1>
    gravityOverride {z=-980}
    physicsBodyDefinitions [{
      boundBone {boneName=dyn_strap_part_fr_01}
      boxExtents=<55 40 40>
      constraintSetup {angularLimitsMax=<5 0 0>, angularLimitsMin=<-5 0 -25>, angularTarget=<1 0 0>}
      localJointOffset=<6 0 0>
     }]
    relativeSpaceBone {boneName=pelvis}
   }
   n32 = Anim Dynamics - Bone: dyn_strap_part_fr_02!(ComponentPose=@n31, Alpha=1, GravityOverride:Vector="0,0,-980", ExternalForce="0,0,0") -> (Pose) {
    angularDampingOverride=0.999
    angularSpringConstant=15000
    bAngularSpring=true
    bOverrideAngularDamping=true
    bUseGravityOverride=true
    gravityScale=0.35
    alphaScaleBiasClamp {bInterpResult=true}
    boundBone {boneName=dyn_strap_part_fr_02}
    componentAppliedLinearAccClamp=<3000 3000 3000>
    componentLinearAccScale=<1 1 1>
    componentLinearVelScale=<1 1 1>
    gravityOverride {z=-980}
    physicsBodyDefinitions [{
      boundBone {boneName=dyn_strap_part_fr_02}
      boxExtents=<55 40 40>
      constraintSetup {angularLimitsMax=<5 0 0>, angularLimitsMin=<-5 0 -25>, angularTarget=<1 0 0>}
      localJointOffset=<6 0 0>
     }]
    relativeSpaceBone {boneName=pelvis}
   }
   n33 = Anim Dynamics - Bone: dyn_strap_part_fr_03!(ComponentPose=@n32, Alpha=1, GravityOverride:Vector="0,0,-980", ExternalForce="0,0,0") -> (Pose) {
    angularDampingOverride=0.999
    angularSpringConstant=15000
    bAngularSpring=true
    bOverrideAngularDamping=true
    bUseGravityOverride=true
    gravityScale=0.35
    alphaScaleBiasClamp {bInterpResult=true}
    boundBone {boneName=dyn_strap_part_fr_03}
    componentAppliedLinearAccClamp=<3000 3000 3000>
    componentLinearAccScale=<1 1 1>
    componentLinearVelScale=<1 1 1>
    gravityOverride {z=-980}
    physicsBodyDefinitions [{
      boundBone {boneName=dyn_strap_part_fr_03}
      boxExtents=<55 40 40>
      constraintSetup {angularLimitsMax=<5 0 0>, angularLimitsMin=<-5 0 -25>, angularTarget=<1 0 0>}
      localJointOffset=<6 0 0>
     }]
    relativeSpaceBone {boneName=pelvis}
   }
   n34 = Anim Dynamics - Bone: dyn_strap_part_fr_04!(ComponentPose=@n33, Alpha=1, GravityOverride:Vector="0,0,-980", ExternalForce="0,0,0") -> (Pose) {
    angularDampingOverride=0.999
    angularSpringConstant=15000
    bAngularSpring=true
    bOverrideAngularDamping=true
    bUseGravityOverride=true
    gravityScale=0.35
    alphaScaleBiasClamp {bInterpResult=true}
    boundBone {boneName=dyn_strap_part_fr_04}
    componentAppliedLinearAccClamp=<3000 3000 3000>
    componentLinearAccScale=<1 1 1>
    componentLinearVelScale=<1 1 1>
    gravityOverride {z=-980}
    physicsBodyDefinitions [{
      boundBone {boneName=dyn_strap_part_fr_04}
      boxExtents=<55 40 40>
      constraintSetup {angularLimitsMax=<5 0 0>, angularLimitsMin=<-5 0 -25>, angularTarget=<1 0 0>}
      localJointOffset=<6 0 0>
     }]
    relativeSpaceBone {boneName=pelvis}
   }
   n35 = Anim Dynamics - Bone: dyn_strap_part_bk_01!(ComponentPose=@n34, Alpha=1, GravityOverride:Vector="0,0,-980", ExternalForce="0,0,0") -> (Pose) {
    angularDampingOverride=0.999
    angularSpringConstant=15000
    bAngularSpring=true
    bOverrideAngularDamping=true
    bUseGravityOverride=true
    gravityScale=0.35
    alphaScaleBiasClamp {bInterpResult=true}
    boundBone {boneName=dyn_strap_part_bk_01}
    componentAppliedLinearAccClamp=<3000 3000 3000>
    componentLinearAccScale=<1 1 1>
    componentLinearVelScale=<1 1 1>
    gravityOverride {z=-980}
    physicsBodyDefinitions [{
      boundBone {boneName=dyn_strap_part_bk_01}
      boxExtents=<55 40 40>
      constraintSetup {angularLimitsMax=<5 0 0>, angularLimitsMin=<-5 0 -25>, angularTarget=<1 0 0>}
      localJointOffset=<6 0 0>
     }]
    relativeSpaceBone {boneName=pelvis}
   }
   n36 = Anim Dynamics - Bone: dyn_strap_part_bk_02!(ComponentPose=@n35, Alpha=1, GravityOverride:Vector="0,0,-980", ExternalForce="0,0,0") -> (Pose) {
    angularDampingOverride=0.999
    angularSpringConstant=15000
    bAngularSpring=true
    bOverrideAngularDamping=true
    bUseGravityOverride=true
    gravityScale=0.35
    alphaScaleBiasClamp {bInterpResult=true}
    boundBone {boneName=dyn_strap_part_bk_02}
    componentAppliedLinearAccClamp=<3000 3000 3000>
    componentLinearAccScale=<1 1 1>
    componentLinearVelScale=<1 1 1>
    gravityOverride {z=-980}
    physicsBodyDefinitions [{
      boundBone {boneName=dyn_strap_part_bk_02}
      boxExtents=<55 40 40>
      constraintSetup {angularLimitsMax=<5 0 0>, angularLimitsMin=<-5 0 -25>, angularTarget=<1 0 0>}
      localJointOffset=<6 0 0>
     }]
    relativeSpaceBone {boneName=pelvis}
   }
   n37 = Anim Dynamics - Bone: dyn_strap_part_bk_03!(ComponentPose=@n36, Alpha=1, GravityOverride:Vector="0,0,-980", ExternalForce="0,0,0") -> (Pose) {
    angularDampingOverride=0.999
    angularSpringConstant=15000
    bAngularSpring=true
    bOverrideAngularDamping=true
    bUseGravityOverride=true
    gravityScale=0.35
    alphaScaleBiasClamp {bInterpResult=true}
    boundBone {boneName=dyn_strap_part_bk_03}
    componentAppliedLinearAccClamp=<3000 3000 3000>
    componentLinearAccScale=<1 1 1>
    componentLinearVelScale=<1 1 1>
    gravityOverride {z=-980}
    physicsBodyDefinitions [{
      boundBone {boneName=dyn_strap_part_bk_03}
      boxExtents=<55 40 40>
      constraintSetup {angularLimitsMax=<5 0 0>, angularLimitsMin=<-5 0 -25>, angularTarget=<1 0 0>}
      localJointOffset=<6 0 0>
     }]
    relativeSpaceBone {boneName=pelvis}
   }
   n38 = Anim Dynamics - Bone: dyn_strap_part_bk_04!(ComponentPose=@n37, Alpha=1, GravityOverride:Vector="0,0,-980", ExternalForce="0,0,0") -> (Pose) {
    angularDampingOverride=0.999
    angularSpringConstant=15000
    bAngularSpring=true
    bOverrideAngularDamping=true
    bUseGravityOverride=true
    gravityScale=0.35
    alphaScaleBiasClamp {bInterpResult=true}
    boundBone {boneName=dyn_strap_part_bk_04}
    componentAppliedLinearAccClamp=<3000 3000 3000>
    componentLinearAccScale=<1 1 1>
    componentLinearVelScale=<1 1 1>
    gravityOverride {z=-980}
    physicsBodyDefinitions [{
      boundBone {boneName=dyn_strap_part_bk_04}
      boxExtents=<55 40 40>
      constraintSetup {angularLimitsMax=<5 0 0>, angularLimitsMin=<-5 0 -25>, angularTarget=<1 0 0>}
      localJointOffset=<6 0 0>
     }]
    relativeSpaceBone {boneName=pelvis}
   }
   n39 = Look At - Bone: base_lollipop(ComponentPose=@n38, Alpha=1) {
    boneToModify {boneName=base_lollipop}
    lookAtLocation=<36.1464 6.0814 3.6949>
    lookAtTarget {boneReference {boneName=dyn_skirt_sd_01_r}}
    lookAt_Axis {axis=<1 0>}
   }
   n40 = Look At - Bone: base_lollipop(ComponentPose=@n39, Alpha=1) {
    boneToModify {boneName=base_lollipop}
    lookAtLocation=<36.1464 9.4417 1.9031>
    lookAtTarget {boneReference {boneName=dyn_skirt_sd_01_r}}
    lookAt_Axis {axis=<1 0>}
   }
   n41 = "Transform (Modify) Bone - Bone: base_lollipop"(ComponentPose=@n40, Alpha=1, Translation="0,0,0", Rotation="0.752,0.157,11.654", Scale="1,1,1") {
    rotationMode=BMM_Additive
    rotationSpace=BCS_BoneSpace
    scaleMode=BMM_Additive
    scaleSpace=BCS_BoneSpace
    translationMode=BMM_Additive
    translationSpace=BCS_BoneSpace
    alphaBoolBlend {blendInTime=0.1, blendOutTime=0.2}
    alphaScaleBiasClamp {bInterpResult=true}
    boneToModify {boneName=base_lollipop}
    rotation=<rot 0.752 0.157 11.654>
   }
   n42 = Anim Dynamics - Bone: dyn_lollipop_01_r!(ComponentPose=@n41, Alpha=1, GravityOverride:Vector="0,0,-980", ExternalForce="0,0,0") -> (Pose) {
    angularDampingOverride=1
    angularSpringConstant=15000
    bAngularSpring=true
    bOverrideAngularDamping=true
    bUseGravityOverride=true
    gravityScale=0.35
    simulationSpace=RootRelative
    boundBone {boneName=dyn_lollipop_01_r}
    componentAppliedLinearAccClamp=<3000 3000 3000>
    componentLinearAccScale=<0.7 0.5 0.5>
    componentLinearVelScale=<0.7 0.5 0.5>
    gravityOverride {z=-980}
    physicsBodyDefinitions [{
      boundBone {boneName=dyn_lollipop_01_r}
      boxExtents=<30 50 30>
      constraintSetup {angularLimitsMax=<2 10 5>, angularLimitsMin=<-2 0 -5>, angularTarget=<1 0 0>}
      localJointOffset=<4 0 0>
     }]
   }
   n43 = Anim Dynamics - Bone: dyn_lollipop_02_r!(ComponentPose=@n42, Alpha=1, GravityOverride:Vector="0,0,-980", ExternalForce="0,0,0") -> (Pose) {
    angularDampingOverride=1
    angularSpringConstant=15000
    bAngularSpring=true
    bOverrideAngularDamping=true
    bUseGravityOverride=true
    gravityScale=0.35
    simulationSpace=RootRelative
    boundBone {boneName=dyn_lollipop_02_r}
    componentAppliedLinearAccClamp=<3000 3000 3000>
    componentLinearAccScale=<0.7 0.5 1>
    componentLinearVelScale=<0.7 0.5 1>
    gravityOverride {z=-980}
    physicsBodyDefinitions [{
      boundBone {boneName=dyn_lollipop_02_r}
      boxExtents=<30 50 30>
      constraintSetup {angularLimitsMax=<2 25 15>, angularLimitsMin=<-2 0 -15>, angularTarget=<1 0 0>}
      localJointOffset=<4 0 0>
     }]
   }
   n44 = "Constraint - dyn_skirt_bk_02_c (RR)"!(ComponentPose=@n43, bAlphaBoolEnabled:bool=True, ConstraintWeights_0=0.5, ConstraintWeights_1:float=0.5) -> (Pose) {
    alphaInputType=Bool
    boneToModify {boneName=dyn_skirt_bk_02_c}
    constraintSetup [{transformType=Rotation, targetBone {boneName=dyn_skirt_bk_02_l}}, {transformType=Rotation, targetBone {boneName=dyn_skirt_bk_02_r}}]
    constraintWeights [1,1]
   }
   n45 = "Constraint - dyn_skirt_bk_03_c (RR)"!(ComponentPose=@n44, bAlphaBoolEnabled:bool=True, ConstraintWeights_0=0.5, ConstraintWeights_1:float=0.5) -> (Pose) {
    alphaInputType=Bool
    boneToModify {boneName=dyn_skirt_bk_03_c}
    constraintSetup [{transformType=Rotation, targetBone {boneName=dyn_skirt_bk_03_l}}, {transformType=Rotation, targetBone {boneName=dyn_skirt_bk_03_r}}]
    constraintWeights [1,1]
   }
   n46 = "Constraint - dyn_skirt_bk_04_c (RR)"!(ComponentPose=@n45, bAlphaBoolEnabled:bool=True, ConstraintWeights_0=0.5, ConstraintWeights_1:float=0.5) -> (Pose) {
    alphaInputType=Bool
    boneToModify {boneName=dyn_skirt_bk_04_c}
    constraintSetup [{transformType=Rotation, targetBone {boneName=dyn_skirt_bk_04_l}}, {transformType=Rotation, targetBone {boneName=dyn_skirt_bk_04_r}}]
    constraintWeights [1,1]
   }
   n47 = "Constraint - dyn_skirt_fr_02_c (RR)"!(ComponentPose=@n46, Alpha=1, ConstraintWeights_0=0.5, ConstraintWeights_1:float=0.5) -> (Pose) {
    alphaScaleBiasClamp {bInterpResult=true, interpSpeedDecreasing=5, interpSpeedIncreasing=5}
    boneToModify {boneName=dyn_skirt_fr_02_c}
    constraintSetup [{transformType=Rotation, targetBone {boneName=dyn_skirt_fr_02_l}}, {transformType=Rotation, targetBone {boneName=dyn_skirt_fr_02_r}}]
    constraintWeights [1,1]
   }
   n48 = "Constraint - dyn_skirt_fr_03_c (RR)"!(ComponentPose=@n47, Alpha=1, ConstraintWeights_0=0.5, ConstraintWeights_1:float=0.5) -> (Pose) {
    alphaScaleBiasClamp {bInterpResult=true, interpSpeedDecreasing=5, interpSpeedIncreasing=5}
    boneToModify {boneName=dyn_skirt_fr_03_c}
    constraintSetup [{transformType=Rotation, targetBone {boneName=dyn_skirt_fr_03_l}}, {transformType=Rotation, targetBone {boneName=dyn_skirt_fr_03_r}}]
    constraintWeights [1,1]
   }
   n49 = "Constraint - dyn_skirt_fr_04_c (RR)"!(ComponentPose=@n48, Alpha=1, ConstraintWeights_0=0.5, ConstraintWeights_1:float=0.5) -> (Pose) {
    alphaScaleBiasClamp {bInterpResult=true, interpSpeedDecreasing=5, interpSpeedIncreasing=5}
    boneToModify {boneName=dyn_skirt_fr_04_c}
    constraintSetup [{transformType=Rotation, targetBone {boneName=dyn_skirt_fr_04_l}}, {transformType=Rotation, targetBone {boneName=dyn_skirt_fr_04_r}}]
    constraintWeights [1,1]
   }
   n50 = "Constraint - dyn_skirt_fr_02_c (RR)"(ComponentPose=@n49, Alpha=1, ConstraintWeights_0=0.5, ConstraintWeights_1=0.5) {
    alphaScaleBiasClamp {bInterpResult=true, interpSpeedDecreasing=5, interpSpeedIncreasing=5}
    boneToModify {boneName=dyn_skirt_fr_02_c}
    constraintSetup [{transformType=Rotation, targetBone {boneName=dyn_skirt_fr_02_l}}, {transformType=Rotation, targetBone {boneName=dyn_skirt_fr_02_r}}]
    constraintWeights [1,1]
   }
   n51 = "Constraint - dyn_skirt_fr_03_c (RR)"(ComponentPose=@n50, Alpha=1, ConstraintWeights_0=0.5, ConstraintWeights_1=0.5) {
    alphaScaleBiasClamp {bInterpResult=true, interpSpeedDecreasing=5, interpSpeedIncreasing=5}
    boneToModify {boneName=dyn_skirt_fr_03_c}
    constraintSetup [{transformType=Rotation, targetBone {boneName=dyn_skirt_fr_03_l}}, {transformType=Rotation, targetBone {boneName=dyn_skirt_fr_03_r}}]
    constraintWeights [1,1]
   }
   n52 = "Constraint - dyn_skirt_fr_04_c (RR)"(ComponentPose=@n51, Alpha=1, ConstraintWeights_0=0.5, ConstraintWeights_1=0.5) {
    alphaScaleBiasClamp {bInterpResult=true, interpSpeedDecreasing=5, interpSpeedIncreasing=5}
    boneToModify {boneName=dyn_skirt_fr_04_c}
    constraintSetup [{transformType=Rotation, targetBone {boneName=dyn_skirt_fr_04_l}}, {transformType=Rotation, targetBone {boneName=dyn_skirt_fr_04_r}}]
    constraintWeights [1,1]
   }
   n53 = "Transform (Modify) Bone - Bone: attach_cape"(ComponentPose=@n52, Alpha=1, Translation="0,-5,2", Rotation="0,0,0", Scale="1,1,1") {
    rotationMode=BMM_Additive
    rotationSpace=BCS_BoneSpace
    scaleMode=BMM_Additive
    scaleSpace=BCS_BoneSpace
    translationMode=BMM_Additive
    translationSpace=BCS_BoneSpace
    alphaBoolBlend {blendInTime=0.2, blendOutTime=0.2}
    alphaScaleBiasClamp {bInterpResult=true}
    boneToModify {boneName=attach_cape}
    translation {y=-5, z=2}
   }
   n54 = "Transform (Modify) Bone - Bone: dyn_strap_part_fr_02"(ComponentPose=@n53, Alpha=1, Translation="0.050,-0.467,-0.893", Rotation="9.187,11.466,11.564", Scale="1,1,1") {
    rotationMode=BMM_Additive
    rotationSpace=BCS_BoneSpace
    translationMode=BMM_Additive
    translationSpace=BCS_BoneSpace
    alphaScaleBiasClamp {bInterpResult=true}
    boneToModify {boneName=dyn_strap_part_fr_02}
    rotation=<rot 9.187 11.466 11.564>
    translation=<0.05 -0.4666 -0.8931>
   }
   n55 = "Transform (Modify) Bone - Bone: base_skirt_fr_c"!(ComponentPose=@n54, bAlphaBoolEnabled:bool=True, Translation="0,0,0", Rotation="5,0,0", Scale="1,1,1") -> (Pose) {
    alphaInputType=Bool
    rotationMode=BMM_Additive
    rotationSpace=BCS_BoneSpace
    translationMode=BMM_Additive
    translationSpace=BCS_BoneSpace
    alphaBoolBlend {blendInTime=0.5, blendOutTime=0.5}
    alphaScaleBiasClamp {bInterpResult=true}
    boneToModify {boneName=base_skirt_fr_c}
    rotation {pitch=5}
   }
   n56 = "Transform (Modify) Bone - Bone: base_skirt_fr_l"!(ComponentPose=@n55, bAlphaBoolEnabled=True, Translation:Vector="0,0,0", Rotation="5,0,0", Scale:Vector="1,1,1") -> (Pose) {
    alphaInputType=Bool
    rotationMode=BMM_Additive
    rotationSpace=BCS_BoneSpace
    translationMode=BMM_Additive
    translationSpace=BCS_BoneSpace
    alphaBoolBlend {blendInTime=0.5, blendOutTime=0.5}
    alphaScaleBiasClamp {bInterpResult=true}
    boneToModify {boneName=base_skirt_fr_l}
    rotation {pitch=5}
   }
   n57 = "Transform (Modify) Bone - Bone: base_skirt_fr_r"!(ComponentPose=@n56, bAlphaBoolEnabled=True, Translation:Vector="0,0,0", Rotation="5,0,0", Scale:Vector="1,1,1") -> (Pose) {
    alphaInputType=Bool
    rotationMode=BMM_Additive
    rotationSpace=BCS_BoneSpace
    translationMode=BMM_Additive
    translationSpace=BCS_BoneSpace
    alphaBoolBlend {blendInTime=0.5, blendOutTime=0.5}
    alphaScaleBiasClamp {bInterpResult=true}
    boneToModify {boneName=base_skirt_fr_r}
    rotation {pitch=5}
   }
   n58 = "Transform (Modify) Bone - Bone: dyn_pending_fr_02_l"(ComponentPose=@n57, Alpha=1, Translation="0,0,0", Rotation="-22.526,0.042,-0.030", Scale="1,1,1") {
    rotationMode=BMM_Additive
    rotationSpace=BCS_BoneSpace
    translationMode=BMM_Additive
    translationSpace=BCS_BoneSpace
    alphaBoolBlend {blendInTime=0.5, blendOutTime=0.5}
    alphaScaleBiasClamp {bInterpResult=true}
    boneToModify {boneName=dyn_pending_fr_02_l}
    rotation=<rot -22.526 0.042 -0.03>
   }
   Save cached pose 'StandingCache'(Pose=Component To Local(ComponentPose=@n58)) {cacheName=StandingCache}
   n59 = "Transform (Modify) Bone - Bone: base_strap_l"!(
    ComponentPose=Local To Component(LocalPose=Use cached pose 'ControlRigCache'())
    bAlphaBoolEnabled:bool=True
    Translation="-0.030,0.035,-0.232"
    Rotation="-9.055,-6.486,-7.769"
    Scale="1,1,1"
   ) -> (Pose) {
    alphaInputType=Bool
    rotationMode=BMM_Additive
    rotationSpace=BCS_BoneSpace
    translationMode=BMM_Additive
    translationSpace=BCS_BoneSpace
    alphaBoolBlend {blendInTime=0.2, blendOutTime=0.2}
    alphaScaleBiasClamp {bInterpResult=true, interpSpeedIncreasing=2}
    boneToModify {boneName=base_strap_l}
    rotation=<rot -9.055 -6.486 -7.769>
    translation=<-0.03 0.035 -0.232>
   }
   n60 = Anim Dynamics - Bone: base_strap_l(ComponentPose=@n59, Alpha=1, GravityOverride="0,0,-980", ExternalForce="0,0,0") {
    angularDampingOverride=0.9
    angularSpringConstant=15000
    bAngularSpring=true
    bOverrideAngularDamping=true
    bUseGravityOverride=true
    gravityScale=0.35
    simulationSpace=RootRelative
    boundBone {boneName=base_strap_l}
    componentAppliedLinearAccClamp=<2500 2500 2500>
    componentLinearAccScale=<0.7 0.7 0.7>
    componentLinearVelScale=<0.7 0.7 0.7>
    gravityOverride {z=-980}
    physicsBodyDefinitions [{
      boundBone {boneName=base_strap_l}
      boxExtents=<30 30 30>
      constraintSetup {angularLimitsMax=<0 0 5>, angularLimitsMin=<0 -40 -5>, angularTarget=<1 0 0>}
      localJointOffset=<6 0 0>
     }]
    simSpaceSettings {simSpaceAngularAlpha=0.5}
   }
   n61 = RigidBody!(ComponentPose=@n60, Alpha, OverrideWorldGravity="0,0,0", ExternalForce) -> (Pose) {
    bOverrideWorldGravity=true
    overridePhysicsAsset=&/F_MED_WaltzScout_PhysicsCrouch
    simulationSpace=BaseBoneSpace
    baseBoneRef {boneName=thigh_l}
    componentAppliedLinearAccClamp=<3000 3000 3000>
    simSpaceSettings {
     maxAngularAcceleration=100
     maxAngularVelocity=15
     maxLinearAcceleration=1000
     maxLinearVelocity=700
     worldAlpha=0.8
     externalLinearDragV=<0.1 0.1 0.1>
    }
   }
   n62 = RigidBody(ComponentPose=@n61, OverrideWorldGravity="0,0,0") {
    bOverrideWorldGravity=true
    overridePhysicsAsset=&/F_MED_WaltzScout_PhysicsCrouchRight
    simulationSpace=BaseBoneSpace
    baseBoneRef {boneName=thigh_r}
    componentAppliedLinearAccClamp=<3000 3000 3000>
    simSpaceSettings {
     maxAngularAcceleration=100
     maxAngularVelocity=15
     maxLinearAcceleration=1000
     maxLinearVelocity=700
     worldAlpha=0.8
     externalLinearDragV=<0.1 0.1 0.1>
    }
   }
   n63 = RigidBody(ComponentPose=@n62, OverrideWorldGravity="0,0,0") {
    bOverrideWorldGravity=true
    overridePhysicsAsset=&F_MED_WaltzScout_PhysicsStrap
    simulationSpace=BaseBoneSpace
    alphaScaleBiasClamp {bInterpResult=true}
    baseBoneRef {boneName=pelvis}
    componentAppliedLinearAccClamp=<3000 3000 3000>
    simSpaceSettings {
     maxAngularAcceleration=200
     maxAngularVelocity=30
     maxLinearAcceleration=2000
     maxLinearVelocity=300
     worldAlpha=1
     externalLinearDragV=<2 2 2>
    }
   }
   n64 = Anim Dynamics - Bone: dyn_lollipop_01_r(ComponentPose=@n63, Alpha=1, GravityOverride="0,0,-980", ExternalForce="0,0,0") {
    angularDampingOverride=1
    angularSpringConstant=15000
    bAngularSpring=true
    bOverrideAngularDamping=true
    bUseGravityOverride=true
    gravityScale=0.35
    simulationSpace=RootRelative
    boundBone {boneName=dyn_lollipop_01_r}
    componentAppliedLinearAccClamp=<3000 3000 3000>
    componentLinearAccScale=<0.7 0.5 0.5>
    componentLinearVelScale=<0.7 0.5 0.5>
    gravityOverride {z=-980}
    physicsBodyDefinitions [{
      boundBone {boneName=dyn_lollipop_01_r}
      boxExtents=<30 50 30>
      constraintSetup {angularLimitsMax=<2 10 5>, angularLimitsMin=<-2 0 -5>, angularTarget=<1 0 0>}
      localJointOffset=<4 0 0>
     }]
   }
   n65 = Anim Dynamics - Bone: dyn_lollipop_02_r(ComponentPose=@n64, Alpha=1, GravityOverride="0,0,-980", ExternalForce="0,0,0") {
    angularDampingOverride=1
    angularSpringConstant=15000
    bAngularSpring=true
    bOverrideAngularDamping=true
    bUseGravityOverride=true
    gravityScale=0.35
    simulationSpace=RootRelative
    boundBone {boneName=dyn_lollipop_02_r}
    componentAppliedLinearAccClamp=<3000 3000 3000>
    componentLinearAccScale=<0.7 0.5 1>
    componentLinearVelScale=<0.7 0.5 1>
    gravityOverride {z=-980}
    physicsBodyDefinitions [{
      boundBone {boneName=dyn_lollipop_02_r}
      boxExtents=<30 50 30>
      constraintSetup {angularLimitsMax=<2 25 15>, angularLimitsMin=<-2 0 -15>, angularTarget=<1 0 0>}
      localJointOffset=<4 0 0>
     }]
   }
   n66 = "Transform (Modify) Bone - Bone: base_skirt_fr_r"(ComponentPose=@n65, bAlphaBoolEnabled=True, Translation="0,0,0", Rotation="1.417,4.898,0.234", Scale="1,1,1") {
    alphaInputType=Bool
    rotationMode=BMM_Additive
    rotationSpace=BCS_BoneSpace
    scaleMode=BMM_Additive
    scaleSpace=BCS_BoneSpace
    translationMode=BMM_Additive
    translationSpace=BCS_BoneSpace
    alphaBoolBlend {blendInTime=0.1, blendOutTime=0.2}
    boneToModify {boneName=base_skirt_fr_r}
    rotation=<rot 1.4171 4.898 0.234>
   }
   n67 = "Transform (Modify) Bone - Bone: base_strap_part_fr"!(ComponentPose=@n66, bAlphaBoolEnabled:bool=True, Translation="0,0,0", Rotation="0.033,-7.881,-13.961", Scale="1,1,1") -> (Pose) {
    alphaInputType=Bool
    rotationMode=BMM_Additive
    rotationSpace=BCS_BoneSpace
    scaleMode=BMM_Additive
    scaleSpace=BCS_BoneSpace
    translationMode=BMM_Additive
    translationSpace=BCS_BoneSpace
    alphaBoolBlend {blendInTime=0.1, blendOutTime=0.2}
    boneToModify {boneName=base_strap_part_fr}
    rotation=<rot 0.033 -7.881 -13.961>
   }
   n68 = "Transform (Modify) Bone - Bone: base_strap_part_fr"!(ComponentPose=@n67, Alpha:float=1, Translation="0,0,0", Rotation="7.034,10.175,10.631", Scale="1,1,1") -> (Pose) {
    rotationMode=BMM_Additive
    rotationSpace=BCS_BoneSpace
    translationMode=BMM_Additive
    translationSpace=BCS_BoneSpace
    alphaBoolBlend {blendInTime=0.2, blendOutTime=0.2}
    alphaScaleBiasClamp {bInterpResult=true, interpSpeedIncreasing=5}
    boneToModify {boneName=base_strap_part_fr}
    rotation=<rot 7.034 10.175 10.631>
   }
   n69 = "Transform (Modify) Bone - Bone: dyn_strap_part_fr_03"(ComponentPose=@n68, Alpha=1, Translation="-0.045,0.647,-0.020", Rotation="-0.099,3.205,9.989", Scale="1,1,1") {
    rotationMode=BMM_Additive
    rotationSpace=BCS_BoneSpace
    translationMode=BMM_Additive
    translationSpace=BCS_BoneSpace
    alphaBoolBlend {blendInTime=0.2, blendOutTime=0.2}
    alphaScaleBiasClamp {bInterpResult=true, interpSpeedIncreasing=5}
    boneToModify {boneName=dyn_strap_part_fr_03}
    rotation=<rot -0.099 3.205 9.989>
    translation=<-0.045 0.647 -0.02>
   }
   n70 = "Transform (Modify) Bone - Bone: dyn_strap_part_bk_03"(ComponentPose=@n69, Alpha=1, Translation="-0.045,0.647,-0.020", Rotation="-0.858,17.913,0.585", Scale="1,1,1") {
    rotationMode=BMM_Additive
    rotationSpace=BCS_BoneSpace
    translationMode=BMM_Additive
    translationSpace=BCS_BoneSpace
    alphaBoolBlend {blendInTime=0.2, blendOutTime=0.2}
    alphaScaleBiasClamp {bInterpResult=true, interpSpeedIncreasing=5}
    boneToModify {boneName=dyn_strap_part_bk_03}
    rotation=<rot -0.858 17.913 0.585>
    translation=<-0.045 0.647 -0.02>
   }
   n71 = "Transform (Modify) Bone - Bone: dyn_strap_part_fr_03"!(ComponentPose=@n70, bAlphaBoolEnabled:bool=True, Translation="-0.102,1.217,-0.020", Rotation="9.996,-3.449,-0.404", Scale="1,1,1") -> (Pose) {
    alphaInputType=Bool
    rotationMode=BMM_Additive
    rotationSpace=BCS_BoneSpace
    translationMode=BMM_Additive
    translationSpace=BCS_BoneSpace
    alphaBoolBlend {blendInTime=0.2, blendOutTime=0.2}
    alphaScaleBiasClamp {bInterpResult=true, interpSpeedIncreasing=5}
    boneToModify {boneName=dyn_strap_part_fr_03}
    rotation=<rot 9.996 -3.449 -0.404>
    translation=<-0.102 1.217 -0.02>
   }
   n72 = "Transform (Modify) Bone - Bone: dyn_strap_part_bk_02"(ComponentPose=@n71, Alpha=1, Translation="0,0,0", Rotation="0.000,-11.846,-0.000", Scale="1,1,1") {
    rotationMode=BMM_Additive
    rotationSpace=BCS_BoneSpace
    translationMode=BMM_Additive
    translationSpace=BCS_BoneSpace
    alphaBoolBlend {blendInTime=0.2, blendOutTime=0.2}
    alphaScaleBiasClamp {bInterpResult=true}
    boneToModify {boneName=dyn_strap_part_bk_02}
    rotation=<rot 0 -11.846 0>
   }
   n73 = "Transform (Modify) Bone - Bone: dyn_skirt_fr_04_l"(ComponentPose=@n72, Alpha=1, Translation="-2.550,0.916,0.328", Rotation="0,0,0", Scale="1,1,1") {
    rotationMode=BMM_Additive
    rotationSpace=BCS_BoneSpace
    scaleMode=BMM_Additive
    scaleSpace=BCS_BoneSpace
    translationMode=BMM_Additive
    translationSpace=BCS_BoneSpace
    alphaBoolBlend {blendInTime=0.1, blendOutTime=0.2}
    alphaScaleBiasClamp {bInterpResult=true}
    boneToModify {boneName=dyn_skirt_fr_04_l}
    translation=<-2.55 0.916 0.328>
   }
   n74 = "Transform (Modify) Bone - Bone: dyn_skirt_fr_04_r"(ComponentPose=@n73, Alpha=1, Translation="-6.219,0,0.434", Rotation="0,0,0", Scale="1,1,1") {
    rotationMode=BMM_Additive
    rotationSpace=BCS_BoneSpace
    scaleMode=BMM_Additive
    scaleSpace=BCS_BoneSpace
    translationMode=BMM_Additive
    translationSpace=BCS_BoneSpace
    alphaBoolBlend {blendInTime=0.1, blendOutTime=0.2}
    alphaScaleBiasClamp {bInterpResult=true}
    boneToModify {boneName=dyn_skirt_fr_04_r}
    translation {x=-6.219, z=0.434}
   }
   n75 = "Transform (Modify) Bone - Bone: dyn_skirt_fr_04_c"!(ComponentPose=@n74, Alpha:float=1, Translation="-2.259,0.000,-0.474", Rotation="0.396,-3.378,3.113", Scale="1,1,1") -> (Pose) {
    rotationMode=BMM_Additive
    rotationSpace=BCS_BoneSpace
    scaleMode=BMM_Additive
    scaleSpace=BCS_BoneSpace
    translationMode=BMM_Additive
    translationSpace=BCS_BoneSpace
    alphaBoolBlend {blendInTime=0.1, blendOutTime=0.2}
    alphaScaleBiasClamp {bInterpResult=true}
    boneToModify {boneName=dyn_skirt_fr_04_c}
    rotation=<rot 0.396 -3.378 3.113>
    translation {x=-2.259, z=-0.474}
   }
   n76 = "Transform (Modify) Bone - Bone: dyn_skirt_sd_04_l"(ComponentPose=@n75, Alpha=1, Translation="-5,0,0", Rotation="0,0,0", Scale="1,1,1") {
    rotationMode=BMM_Additive
    rotationSpace=BCS_BoneSpace
    scaleMode=BMM_Additive
    scaleSpace=BCS_BoneSpace
    translationMode=BMM_Additive
    translationSpace=BCS_BoneSpace
    alphaBoolBlend {blendInTime=0.1, blendOutTime=0.2}
    alphaScaleBiasClamp {bInterpResult=true}
    boneToModify {boneName=dyn_skirt_sd_04_l}
    translation {x=-5}
   }
   n77 = "Transform (Modify) Bone - Bone: dyn_skirt_sd_04_r"(ComponentPose=@n76, Alpha=1, Translation="-5.139,-0.000,-0.052", Rotation="0,0,0", Scale="1,1,1") {
    rotationMode=BMM_Additive
    rotationSpace=BCS_BoneSpace
    scaleMode=BMM_Additive
    scaleSpace=BCS_BoneSpace
    translationMode=BMM_Additive
    translationSpace=BCS_BoneSpace
    alphaBoolBlend {blendInTime=0.1, blendOutTime=0.2}
    alphaScaleBiasClamp {bInterpResult=true}
    boneToModify {boneName=dyn_skirt_sd_04_r}
    translation {x=-5.139, z=-0.052}
   }
   n78 = "Constraint - dyn_skirt_bk_02_c (RR)"(ComponentPose=@n77, bAlphaBoolEnabled=True, ConstraintWeights_0=0.5, ConstraintWeights_1=0.5) {
    alphaInputType=Bool
    boneToModify {boneName=dyn_skirt_bk_02_c}
    constraintSetup [{transformType=Rotation, targetBone {boneName=dyn_skirt_bk_02_l}}, {transformType=Rotation, targetBone {boneName=dyn_skirt_bk_02_r}}]
    constraintWeights [1,1]
   }
   n79 = "Constraint - dyn_skirt_bk_03_c (RR)"(ComponentPose=@n78, bAlphaBoolEnabled=True, ConstraintWeights_0=0.5, ConstraintWeights_1=0.5) {
    alphaInputType=Bool
    boneToModify {boneName=dyn_skirt_bk_03_c}
    constraintSetup [{transformType=Rotation, targetBone {boneName=dyn_skirt_bk_03_l}}, {transformType=Rotation, targetBone {boneName=dyn_skirt_bk_03_r}}]
    constraintWeights [1,1]
   }
   n80 = "Constraint - dyn_skirt_bk_04_c (RR)"(ComponentPose=@n79, bAlphaBoolEnabled=True, ConstraintWeights_0=0.5, ConstraintWeights_1=0.5) {
    alphaInputType=Bool
    boneToModify {boneName=dyn_skirt_bk_04_c}
    constraintSetup [{transformType=Rotation, targetBone {boneName=dyn_skirt_bk_04_l}}, {transformType=Rotation, targetBone {boneName=dyn_skirt_bk_04_r}}]
    constraintWeights [1,1]
   }
   n81 = "Transform (Modify) Bone - Bone: base_skirt_fr_l"!(ComponentPose=@n80, Alpha:float=1, Translation="0.002,-0.020,0.049", Rotation="-1.786,0.995,-0.332", Scale="1,1,1") -> (Pose) {
    rotationMode=BMM_Additive
    rotationSpace=BCS_BoneSpace
    translationMode=BMM_Additive
    translationSpace=BCS_BoneSpace
    alphaScaleBiasClamp {bInterpResult=true, interpSpeedIncreasing=2}
    boneToModify {boneName=base_skirt_fr_l}
    rotation=<rot -1.786 0.995 -0.332>
    translation=<0.002 -0.02 0.049>
   }
   n82 = "Transform (Modify) Bone - Bone: base_skirt_sd_l"(ComponentPose=@n81, Alpha=1, Translation="0,0,0", Rotation="-4.120,6.167,-3.357", Scale="1,1,1") {
    rotationMode=BMM_Additive
    rotationSpace=BCS_BoneSpace
    translationMode=BMM_Additive
    translationSpace=BCS_BoneSpace
    alphaScaleBiasClamp {bInterpResult=true, interpSpeedIncreasing=2}
    boneToModify {boneName=base_skirt_sd_l}
    rotation=<rot -4.12 6.167 -3.357>
   }
   n83 = "Transform (Modify) Bone - Bone: base_skirt_bk_l"(ComponentPose=@n82, Alpha=1, Translation="0,0,0", Rotation="2.308,2.928,4.275", Scale="1,1,1") {
    rotationMode=BMM_Additive
    rotationSpace=BCS_BoneSpace
    translationMode=BMM_Additive
    translationSpace=BCS_BoneSpace
    alphaScaleBiasClamp {bInterpResult=true, interpSpeedIncreasing=2}
    boneToModify {boneName=base_skirt_bk_l}
    rotation=<rot 2.308 2.928 4.275>
   }
   Save cached pose 'CrouchCache'(Pose=Component To Local(ComponentPose=@n83)) {cacheName=CrouchCache}
  ]}
 EventGraph {nodes [
   Event BlueprintUpdateAnimation()
   > n1 = Sequence!(execute) -> (then_0, then_1, then_2:Exec, then_3:Exec, then_4:Exec, then_5:Exec)
   > n2 = "Gravity Override State Function (Parameter Asset) (Nativized)"(
    CharacterPartAnimInstance=Self-Reference()
    GameTimeInSeconds=GetGameTimeInSeconds()
    DebugJointName=spine_05
    Parameters=&/F_MED_WaltzScoutParameters
   )
   > Set Gravity Override(Gravity Override=@n2)
   > Set RBAN Settings Out(RBAN Settings Out=@n2.OutRigidBodyAnimNodeParameters)
   n3 = GetAnimInstance(self=GetOwnerSkeletalMeshComponent())
   > Set BoolPlayingEasternBloc(BoolPlayingEasternBloc:bool=Montage_IsPlaying(self=@n3, Montage=/BRCosmetics/Animation/Game/MainPlayer/Emotes/Eastern_Bloc/Emote_EasternBloc_CMF_M)) -> (Output_Get:bool)
   > Set Alpha_PlayingSharpfang(Alpha_PlayingSharpfang:bool=OR Boolean(
     A=Montage_IsPlaying(self=@n3, Montage=/BRCosmetics/Animation/Game/MainPlayer/Emotes/SharpFang/Emote_SharpFang_CMF_Skirt_M_FrontEnd)
     B=Montage_IsPlaying(self=@n3, Montage=/BRCosmetics/Animation/Game/MainPlayer/Emotes/SharpFang/Emote_SharpFang_CMF_Skirt_M)
    )) -> (Output_Get:bool)
   n4 = Branch(execute=@n1.then_2, Condition=Get bIsDBNO())
   > Set DBNO(DBNO:double=1) -> (Output_Get:double)
   > Set NoDBNO(NoDBNO:double=0) -> (Output_Get:double)
   Set DBNO(execute=@n4.else, DBNO=0)
   > Set NoDBNO(NoDBNO=1)
   n5 = Branch(execute=@n1.then_3, Condition=Get bIsPlayingEmote())
   > Set Alpha_ConstraintEmote(Alpha_ConstraintEmote:double=0.5) -> (Output_Get:double)
   Set Alpha_ConstraintEmote(execute=@n5.else, Alpha_ConstraintEmote=0)
   n6 = Get bIsSliding()
   n7 = OR Boolean(A=AND Boolean(A=Get bIsCrouching(), B=NOT Boolean(A=@n6)), B=Get bIsDBNO())
   n8 = Get bRigidBodyEnabled()
   Set CrouchNotSliding(execute=@n1.then_1, CrouchNotSliding:double="To Float (Boolean)"(InBool=AND Boolean(A=@n7, B=@n8))) -> (Output_Get:double)
   n9 = Get bIsSkydiveFloating()
   n10 = Get bIsSkydiveDiveMode()
   > Set FloatingAndSkydiveMode(FloatingAndSkydiveMode:double="To Float (Boolean)"(InBool=OR Boolean(A=@n9, B=@n10))) -> (Output_Get:double)
   n11 = NOT Boolean(A=Get bIsDBNOCarried())
   n12 = NOT Boolean(A=Get bIsMoving2D())
   n13 = NOT Boolean(A=Get bIsDBNOCarrying() -> (bIsDBNOCarrying:bool))
   n14 = Get bIsDBNO()
   > Set DBNOIdle(DBNOIdle:double="To Float (Boolean)"(InBool=AND Boolean!(A=@n12, B=@n14, C:bool=@n13, D:bool=@n11) -> (ReturnValue))) -> (Output_Get:double)
   n15 = NOT Boolean(A=Get bIsDBNO())
   n16 = NOT Boolean(A=Get bIsJumping())
   n17 = NOT Boolean(A=Get bIsPlayingEmote())
   n18 = NOT Boolean(A=Get bIsSkyDiving())
   n19 = NOT Boolean(A=Get bIsMoving2D())
   > Set Idle(Idle:double="To Float (Boolean)"(InBool=AND Boolean!(A=@n19, B=@n18, C=@n17, D=@n15, E:bool=@n16) -> (ReturnValue))) -> (Output_Get:double)
   n20 = NOT Boolean(A=Get bIsMoving2D())
   n21 = NOT Boolean(A=Get bIsSkyDiving())
   n22 = NOT Boolean(A=Get bIsPlayingEmote())
   n23 = Get bIsInFrontEnd()
   > Set Alpha_IdleFE(Alpha_IdleFE:double="To Float (Boolean)"(InBool=AND Boolean!(A=@n23, B=@n20, C=@n21, D=@n22) -> (ReturnValue))) -> (Output_Get:double)
   n24 = NOT Boolean(A=Get bIsPlayingEmote())
   n25 = NOT Boolean(A=Get bIsSkyDiving())
   n26 = Get Detected_Pistol() -> (Detected_Pistol:bool)
   > Set Alpha_Pistol(Alpha_Pistol:double="To Float (Boolean)"(InBool=AND Boolean!(A=@n26, B=@n25, C=@n24) -> (ReturnValue))) -> (Output_Get:double)
   n27 = AND Boolean(A=Get bIsFalling(), B=NOT Boolean(A=Get bIsSkyDiving()), C=Get bRigidBodyEnabled())
   n28 = AND Boolean!(A=Get bIsInFrontEnd(), B=NOT Boolean(A=Get bIsMoving2D()), C=NOT Boolean(A=Get bIsSkyDiving()), D:bool=NOT Boolean(A=Get bIsPlayingEmote())) -> (ReturnValue)
   > Set Alpha_SkirtFRConstraint(Alpha_SkirtFRConstraint:double="To Float (Boolean)"(InBool=OR Boolean(A=@n27, B=@n28))) -> (Output_Get:double)
   n29 = Get bIsInFrontEnd()
   > Set Alpha_NoFrontEnd(Alpha_NoFrontEnd:double="To Float (Boolean)"(InBool=NOT Boolean(A=@n29))) -> (Output_Get:double)
   n30 = NOT Boolean(A=Get bIsSkyDiving())
   n31 = NOT Boolean(A=Get bIsPlayingEmote())
   n32 = Get bIsFalling()
   > Set Alpha_FallingOnGround(Alpha_FallingOnGround:double="To Float (Boolean)"(InBool=AND Boolean!(A=@n32, B=@n30, C=@n31) -> (ReturnValue))) -> (Output_Get:double)
   n33 = NOT Boolean(A=Get bIsDBNOCarried())
   n34 = Get bIsDBNO()
   > n35 = Branch(Condition=OR Boolean!(A=Get bIsOnGrindRail(), B=AND Boolean!(A=@n33, B=@n34) -> (ReturnValue), C:bool=Get BoolPlayingEasternBloc()) -> (ReturnValue))
   > Set RB Alpha(RB Alpha:double=0.75) -> (Output_Get:double)
   Set RB Alpha(execute=@n35.else, RB Alpha=1)
   n36 = GetOwnerSkeletalMeshComponent()
   n37 = BreakRotator(InRot=BreakTransform(InTransform=GetDeltaTransformFromRefPose(self=@n36, BoneName=spine_05, BaseName=pelvis)).Rotation)
   n38 = Branch(
    execute=@n1.then_4
    Condition=AND Boolean!(A="float > float"(A=@n37, B=12), B=Get bIsPlayingMeleeAnim(), C:bool=Get bIsMoving2D()) -> (ReturnValue)
   )
   > Set Alpha_ArmLBackwardMelee(Alpha_ArmLBackwardMelee:double=1) -> (Output_Get:double)
   Set Alpha_ArmLBackwardMelee(execute=@n38.else, Alpha_ArmLBackwardMelee=0)
   Event OnParachuteOpened()
   > Switch on EGliderType(Selection=Get PlayerGliderType() -> (PlayerGliderType=HangGlider))
   > Set HangGlidingAlpha(HangGlidingAlpha:double=1) -> (Output_Get:double)
   Event OnParachuteClosed()
   > Set HangGlidingAlpha(HangGlidingAlpha=0)
   Event OnWeaponChanged()
   > n39 = Cast To FortPlayerAnimInstance(Object=GetAnimInstance(self=GetOwnerSkeletalMeshComponent()))
   > Set Detected_Pistol(Detected_Pistol:bool=false) -> (Output_Get:bool)
   > n40 = Switch on EFortWeaponCoreAnimation!(execute, Selection:EFortWeaponCoreAnimation Enum=Get WeaponCoreAnim(self=@n39)) -> (
    Melee:Exec
    Pistol:Exec
    Shotgun:Exec
    PaperBlueprint:Exec
    Rifle:Exec
    MeleeOneHand:Exec
    MachinePistol:Exec
    RocketLauncher:Exec
    GrenadeLauncher:Exec
    GoingCommando:Exec
    AssaultRifle:Exec
    TacticalShotgun:Exec
    SniperRifle:Exec
    TrapPlacement:Exec
    ShoulderLauncher:Exec
    AbilityDecoTool:Exec
    Crossbow:Exec
    C4:Exec
    RemoteControl:Exec
    DualWield:Exec
    AR_BullPup:Exec
    AR_ForwardGrip:Exec
    MedPackPaddles:Exec
    SMG_P90:Exec
    AR_DrumGun:Exec
    Consumable_Small:Exec
    Consumable_Large:Exec
    Balloon:Exec
    MountedTurret:Exec
    CreativeTool:Exec
    ExplosiveBow:Exec
    AshtonIndigo:Exec
    AshtonChicago:Exec
    MeleeDualWield:Exec
    Unarmed:Exec
   )
   Set Detected_Pistol(execute=@n40.Pistol, Detected_Pistol=true)
   n41 = GetAnimInstance(self=GetOwnerSkeletalMeshComponent())
   Set BoolWaltzScoutEmote(BoolWaltzScoutEmote:bool=OR Boolean!(
     A=Montage_IsPlaying(self=@n41, Montage=/BRCosmetics/Animation/Game/MainPlayer/Emotes/WaltzScout/Skirted/Emote_WaltzScout_Skirt_CMF_M)
     B=Montage_IsPlaying(self=@n41, Montage=/BRCosmetics/Animation/Game/MainPlayer/Emotes/WaltzScout/Skirted/Emote_WaltzScout_Skirt_CMF_M_FE)
     C=Montage_IsPlaying(self=@n41, Montage=/BRCosmetics/Animation/Game/MainPlayer/Emotes/NewsVan/Emote_NewsVan_CMF_M)
     D:bool=Montage_IsPlaying(self=@n41, Montage=/BRCosmetics/Animation/Game/MainPlayer/Emotes/NewsVan/FE/Emote_NewsVan_CMF_FE_M)
    ) -> (ReturnValue)) -> (Output_Get:bool)
  ]}
]
variables [
 Alpha_ArmLBackwardMelee {type=number}
 Alpha_ConstraintEmote {type=number}
 Alpha_FallingOnGround {type=number}
 Alpha_IdleFE {type=number}
 Alpha_NoFrontEnd {type=number}
 Alpha_Pistol {type=number}
 Alpha_PlayingSharpfang {type=boolean}
 Alpha_SkirtFRConstraint {type=number}
 BoolPlayingEasternBloc {type=boolean}
 BoolWaltzScoutEmote {type=boolean}
 CrouchNotSliding {type=number}
 Custom Idle Is Playing Float {type=number}
 DBNO {type=number}
 DBNOIdle {type=number}
 Detected_Pistol {type=boolean}
 FloatingAndSkydiveMode {type=number}
 Gravity Override {type=Vector}
 HangGlidingAlpha {type=number}
 Idle {type=number}
 Is On Grind Rail {type=boolean}
 NoDBNO {type=number}
 RB Alpha {type=number}
 RBAN Settings Out {type=FortRigidBodyAnimNodeParameters}
]
