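# Physics asset for the female medium WaltzScout face accessory mesh, setting up hair physics for twin ponytails, side strands and a front lock.
# Capsule chains hang from base bones constrained to the head: six links per ponytail, three per strand, two at the front.
# Swing limits widen from stiff ponytail roots toward the tips, with twist mostly locked.
# Head, upper spine, clavicle and upper-arm capsules, joined by free constraints, provide body collision shapes.
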PhysicsAsset /BRCosmetics/Characters/Player/Female/Medium/Bodies/F_MED_WaltzScout/Meshes/Parts/F_MED_WaltzScout_FaceAcc_Physics
bodies {
 base_hair_fr [{shapeType=Capsule, length=0.505, radius=0.505, shapeName=base_hair_fr_capsule, rotation=<rot 0 -90 0>}]
 base_ponytail_l [{shapeType=Capsule, length=0.505, radius=0.505, shapeName=base_ponytail_l_capsule, center=<2 0 0>, rotation=<rot 0 -90 0>}]
 base_ponytail_r [{shapeType=Capsule, length=0.505, radius=0.505, shapeName=base_ponytail_r_capsule, center=<-2 0 0>, rotation=<rot 0 -90 180>}]
 base_strand_l [{shapeType=Capsule, length=0.505, radius=0.505, shapeName=base_strand_l_capsule, rotation=<rot 0 -90 0>}]
 base_strand_r [{shapeType=Capsule, length=0.505, radius=0.505, shapeName=base_strand_r_capsule, rotation=<rot 0 -90 180>}]
 clavicle_l [{
   shapeType=Capsule
   length=5.7046
   radius=6.3469
   shapeName=clavicle_l_capsule
   center=<4.5022 -1.1227 -1.3052>
   rotation=<rot 79.8911 -0.0004 -0.0004>
  }]
 clavicle_r [{
   shapeType=Capsule
   length=5.7046
   radius=6.3469
   shapeName=clavicle_r_capsule
   center=<-4.5022 1.1227 1.3052>
   rotation=<rot 79.8911 -0.0004 179.9996>
  }]
 dyn_hair_fr_01 [{shapeType=Capsule, length=5, radius=1, shapeName=dyn_hair_fr_01_capsule, center=<1.1446 0 0>}]
 dyn_hair_fr_02 [{shapeType=Capsule, length=5, radius=1, shapeName=dyn_hair_fr_02_capsule, center=<1.1617 0 0>}]
 dyn_ponytail_01_l [{shapeType=Capsule, length=4, radius=2.2479, shapeName=dyn_ponytail_01_l_capsule, center=<2 0 0>, rotation=<rot 90 0 0>}]
 dyn_ponytail_01_r [{shapeType=Capsule, length=4, radius=2.2479, shapeName=dyn_ponytail_01_r_capsule, center=<-2 0 0>, rotation=<rot 90 180 0>}]
 dyn_ponytail_02_l [{shapeType=Capsule, length=4, radius=5, shapeName=dyn_ponytail_02_l_capsule, center=<2 0 0>, rotation=<rot 76.8384 -179.9998 -179.9998>}]
 dyn_ponytail_02_r [{shapeType=Capsule, length=4, radius=5, shapeName=dyn_ponytail_02_r_capsule, center=<-2 0 0>, rotation=<rot 76.8384 -179.9998 0.0002>}]
 dyn_ponytail_03_l [{shapeType=Capsule, length=3.0462, radius=5.9538, shapeName=dyn_ponytail_03_l_capsule, center=<2 0.4649 0>, rotation=<rot 90 0 0>}]
 dyn_ponytail_03_r [{shapeType=Capsule, length=3.0462, radius=5.9538, shapeName=dyn_ponytail_03_r_capsule, center=<-2 -0.4649 0>, rotation=<rot 90 180 0>}]
 dyn_ponytail_04_l [{shapeType=Capsule, length=4.2783, radius=5, shapeName=dyn_ponytail_04_l_capsule, center=<2 0 0>, rotation=<rot 90 0 0>}]
 dyn_ponytail_04_r [{shapeType=Capsule, length=4.2783, radius=5, shapeName=dyn_ponytail_04_r_capsule, center=<-2 0 0>, rotation=<rot 90 180 0>}]
 dyn_ponytail_05_l [{shapeType=Capsule, length=4.9883, radius=4.2901, shapeName=dyn_ponytail_05_l_capsule, center=<2 0 -0.5497>, rotation=<rot 90 0 0>}]
 dyn_ponytail_05_r [{shapeType=Capsule, length=4.9883, radius=4.2901, shapeName=dyn_ponytail_05_r_capsule, center=<-2 0 0.5497>, rotation=<rot 90 180 0>}]
 dyn_ponytail_06_l [{
   shapeType=Capsule
   length=6.5387
   radius=2.7396
   shapeName=dyn_ponytail_06_l_capsule
   center=<3.8816 0.5601 -1.5929>
   rotation=<rot 78.9825 117.3393 116.7824>
  }]
 dyn_ponytail_06_r [{
   shapeType=Capsule
   length=6.5387
   radius=2.7396
   shapeName=dyn_ponytail_06_r_capsule
   center=<-3.8816 -0.5601 1.5929>
   rotation=<rot 78.9825 117.3393 -63.2176>
  }]
 dyn_strand_01_l [{
   shapeType=Capsule
   length=0.6043
   radius=1.678
   shapeName=dyn_strand_01_l_capsule
   center=<3.3785 -0.9229 1.4035>
   rotation=<rot 0 0.0001 98.8213>
  }]
 dyn_strand_01_r [{shapeType=Capsule, length=2.0864, radius=1.678, shapeName=dyn_strand_01_r_capsule, center=<-3.3785 0.922 -0.43>, rotation=<rot 0 0 -90>}]
 dyn_strand_02_l [{shapeType=Capsule, length=0.9601, radius=2.2048, shapeName=dyn_strand_02_l_capsule, center=<2 -0.1 0.6>, rotation=<rot 0 0.0001 106.9829>}]
 dyn_strand_02_r [{shapeType=Capsule, length=0.9601, radius=2.2048, shapeName=dyn_strand_02_r_capsule, center=<-2 0.1 -0.6>, rotation=<rot 0 0.0001 -73.0171>}]
 dyn_strand_03_l [{
   shapeType=Capsule
   length=6.7036
   radius=0.8708
   shapeName=dyn_strand_03_l_capsule
   center=<2.1169 0.6061 -1.0118>
   rotation=<rot 70.212 10.4816 -6.4629>
  }]
 dyn_strand_03_r [{
   shapeType=Capsule
   length=6.7036
   radius=0.8708
   shapeName=dyn_strand_03_r_capsule
   center=<-2.1169 -0.6061 1.0118>
   rotation=<rot 70.212 10.4816 173.5371>
  }]
 head [{shapeType=Capsule, length=7.3051, radius=7.8124, center=<2.6019 -1.7469 0>, rotation=<rot 90 0 0>}]
 spine_04 [{shapeType=Capsule, length=9.2354, radius=11.0592, shapeName=spine_04_capsule, center=<-7.7916 3.6903 0>, rotation=<rot 90 0 0>}]
 spine_05 [{shapeType=Capsule, length=7.3858, radius=12.2298, shapeName=spine_05_capsule, center=<0 1.7867 0>, rotation=<rot 0 -90 0>}]
 upperarm_l [{shapeType=Capsule, length=25, radius=6, shapeName=upperarm_l_capsule, center=<14.0563 0 0>, rotation=<rot 90 0 0>}]
 upperarm_r [{shapeType=Capsule, length=25.8285, radius=4.4, shapeName=upperarm_r_capsule, center=<-14.0563 0.434 0>, rotation=<rot 90 180 0>}]
}
constraints [
 {
  bone1Name=base_ponytail_l
  bone2Name=head
  swing1LimitDegrees=45
  swing1Motion=Free
  swing2LimitDegrees=45
  swing2Motion=Free
  twistLimitDegrees=45
  twistMotion=Free
 }
 {bone1Name=dyn_ponytail_01_l, bone2Name=base_ponytail_l, swing1LimitDegrees=0, swing2LimitDegrees=0, twistLimitDegrees=45, twistMotion=Locked}
 {
  bone1Name=dyn_ponytail_02_l
  bone2Name=dyn_ponytail_01_l
  swing1LimitDegrees=2
  swing2LimitDegrees=0
  twistLimitDegrees=45
  twistMotion=Locked
 }
 {
  bone1Name=dyn_ponytail_03_l
  bone2Name=dyn_ponytail_02_l
  swing1LimitDegrees=17
  swing2LimitDegrees=17
  twistLimitDegrees=45
  twistMotion=Locked
 }
 {
  bone1Name=head
  bone2Name=spine_05
  swing1LimitDegrees=45
  swing1Motion=Free
  swing2LimitDegrees=45
  swing2Motion=Free
  twistLimitDegrees=45
  twistMotion=Free
 }
 {
  bone1Name=upperarm_l
  bone2Name=clavicle_l
  swing1LimitDegrees=45
  swing1Motion=Free
  swing2LimitDegrees=45
  swing2Motion=Free
  twistLimitDegrees=45
  twistMotion=Free
 }
 {
  bone1Name=upperarm_r
  bone2Name=clavicle_r
  swing1LimitDegrees=45
  swing1Motion=Free
  swing2LimitDegrees=45
  swing2Motion=Free
  twistLimitDegrees=45
  twistMotion=Free
 }
 {
  bone1Name=spine_05
  bone2Name=spine_04
  swing1LimitDegrees=45
  swing1Motion=Free
  swing2LimitDegrees=45
  swing2Motion=Free
  twistLimitDegrees=45
  twistMotion=Free
 }
 {
  bone1Name=dyn_ponytail_06_l
  bone2Name=dyn_ponytail_05_l
  swing1LimitDegrees=20
  swing2LimitDegrees=20
  twistLimitDegrees=45
  twistMotion=Locked
 }
 {
  bone1Name=dyn_ponytail_05_l
  bone2Name=dyn_ponytail_04_l
  swing1LimitDegrees=20
  swing2LimitDegrees=20
  twistLimitDegrees=45
  twistMotion=Locked
 }
 {
  bone1Name=dyn_ponytail_04_l
  bone2Name=dyn_ponytail_03_l
  swing1LimitDegrees=20
  swing2LimitDegrees=20
  twistLimitDegrees=45
  twistMotion=Locked
 }
 {
  bone1Name=clavicle_l
  bone2Name=spine_05
  swing1LimitDegrees=45
  swing1Motion=Free
  swing2LimitDegrees=45
  swing2Motion=Free
  twistLimitDegrees=45
  twistMotion=Free
 }
 {
  bone1Name=clavicle_r
  bone2Name=spine_05
  swing1LimitDegrees=45
  swing1Motion=Free
  swing2LimitDegrees=45
  swing2Motion=Free
  twistLimitDegrees=45
  twistMotion=Free
 }
 {
  bone1Name=base_strand_l
  bone2Name=head
  swing1LimitDegrees=45
  swing1Motion=Free
  swing2LimitDegrees=45
  swing2Motion=Free
  twistLimitDegrees=45
  twistMotion=Free
 }
 {bone1Name=dyn_strand_01_l, bone2Name=base_strand_l, swing1LimitDegrees=15, swing2LimitDegrees=5, twistLimitDegrees=45, twistMotion=Locked}
 {bone1Name=dyn_strand_02_l, bone2Name=dyn_strand_01_l, swing1LimitDegrees=15, swing2LimitDegrees=15, twistLimitDegrees=45, twistMotion=Locked}
 {bone1Name=dyn_strand_03_l, bone2Name=dyn_strand_02_l, swing1LimitDegrees=25, swing2LimitDegrees=25, twistLimitDegrees=45, twistMotion=Locked}
 {bone1Name=dyn_hair_fr_02, bone2Name=dyn_hair_fr_01, swing1LimitDegrees=3, swing2LimitDegrees=3, twistLimitDegrees=3}
 {
  bone1Name=dyn_hair_fr_01
  bone2Name=base_hair_fr
  swing1LimitDegrees=3
  swing2LimitDegrees=3
  swing2Motion=Locked
  twistLimitDegrees=45
  twistMotion=Locked
 }
 {
  bone1Name=base_hair_fr
  bone2Name=head
  swing1LimitDegrees=45
  swing1Motion=Free
  swing2LimitDegrees=45
  swing2Motion=Free
  twistLimitDegrees=45
  twistMotion=Free
 }
 {
  bone1Name=base_ponytail_r
  bone2Name=head
  swing1LimitDegrees=45
  swing1Motion=Free
  swing2LimitDegrees=45
  swing2Motion=Free
  twistLimitDegrees=45
  twistMotion=Free
 }
 {bone1Name=dyn_ponytail_01_r, bone2Name=base_ponytail_r, swing1LimitDegrees=0, swing2LimitDegrees=0, twistLimitDegrees=45, twistMotion=Locked}
 {
  bone1Name=dyn_ponytail_02_r
  bone2Name=dyn_ponytail_01_r
  swing1LimitDegrees=2
  swing2LimitDegrees=0
  twistLimitDegrees=45
  twistMotion=Locked
 }
 {
  bone1Name=dyn_ponytail_03_r
  bone2Name=dyn_ponytail_02_r
  swing1LimitDegrees=17
  swing2LimitDegrees=17
  twistLimitDegrees=45
  twistMotion=Locked
 }
 {
  bone1Name=dyn_ponytail_04_r
  bone2Name=dyn_ponytail_03_r
  swing1LimitDegrees=20
  swing2LimitDegrees=20
  twistLimitDegrees=45
  twistMotion=Locked
 }
 {
  bone1Name=dyn_ponytail_05_r
  bone2Name=dyn_ponytail_04_r
  swing1LimitDegrees=20
  swing2LimitDegrees=20
  twistLimitDegrees=45
  twistMotion=Locked
 }
 {
  bone1Name=dyn_ponytail_06_r
  bone2Name=dyn_ponytail_05_r
  swing1LimitDegrees=20
  swing2LimitDegrees=20
  twistLimitDegrees=45
  twistMotion=Locked
 }
 {
  bone1Name=base_strand_r
  bone2Name=head
  swing1LimitDegrees=45
  swing1Motion=Free
  swing2LimitDegrees=45
  swing2Motion=Free
  twistLimitDegrees=45
  twistMotion=Free
 }
 {bone1Name=dyn_strand_01_r, bone2Name=base_strand_r, swing1LimitDegrees=15, swing2LimitDegrees=5, twistLimitDegrees=45, twistMotion=Locked}
 {bone1Name=dyn_strand_02_r, bone2Name=dyn_strand_01_r, swing1LimitDegrees=15, swing2LimitDegrees=15, twistLimitDegrees=45, twistMotion=Locked}
 {bone1Name=dyn_strand_03_r, bone2Name=dyn_strand_02_r, swing1LimitDegrees=25, swing2LimitDegrees=25, twistLimitDegrees=45, twistMotion=Locked}
]
