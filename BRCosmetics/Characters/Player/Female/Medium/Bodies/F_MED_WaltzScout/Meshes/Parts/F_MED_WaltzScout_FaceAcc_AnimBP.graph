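# Animation Blueprint for the F_MED_WaltzScout female face accessory, adding secondary motion to ponytails and side hair.
# It copies the parent pose, layers in Body_Mesh arm bones, offsets ponytail base rotations, then runs head-space RigidBody and per-bone AnimDynamics simulation.
# In the front end, the Frontend_Idle_Bypass curve blends DefaultSlot animation over the simulated right ponytail.
# Each update also computes gravity override and rigid-body settings from a parameter asset.
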AnimBlueprint /BRCosmetics/Characters/Player/Female/Medium/Bodies/F_MED_WaltzScout/Meshes/Parts/F_MED_WaltzScout_FaceAcc_AnimBP
target_skeleton=&/F_MED_WaltzScout_FaceAcc_Skeleton
graphs [
 AnimGraph {nodes [
   n1 = Layered blend per bone(
    BasePose=Use cached pose 'OuputAnimGraph'()
    BlendPoses_0=Slot 'DefaultSlot'()
    BlendWeights_0=Get Custom Idle Is Playing Float() -> (Custom Idle Is Playing Float:double)
   ) {layerSetup [{branchFilters [
       {boneName=base_ponytail_r}
       {boneName=dyn_ponytail_01_r}
       {boneName=dyn_ponytail_02_r}
       {boneName=dyn_ponytail_03_r}
       {boneName=dyn_ponytail_04_r}
       {boneName=dyn_ponytail_05_r}
       {boneName=dyn_ponytail_06_r}
      ]}]}
   Output Pose(Result=@n1) {layerGroup=None}
   n2 = Layered blend per bone(
    BasePose=Copy Pose From Mesh() {bUseAttachedParent=true}
    BlendPoses_0=Copy Pose From Mesh(SourceMeshComponent=Get Body_Mesh()) {bUseAttachedParent=true}
   ) {layerSetup [{branchFilters [{boneName=clavicle_l}, {boneName=clavicle_r}, {boneName=upperarm_r}, {boneName=upperarm_l}, {boneName=lowerarm_l}, {boneName=lowerarm_r}]}]}
   n3 = "Transform (Modify) Bone - Bone: base_ponytail_l"!(
    ComponentPose=Local To Component(LocalPose=@n2)
    bAlphaBoolEnabled:bool=True
    Translation="0,0,0"
    Rotation="0.000,0.000,-71.348"
    Scale="1,1,1"
   ) -> (Pose) {
    alphaInputType=Bool
    rotationMode=BMM_Additive
    rotationSpace=BCS_BoneSpace
    scaleMode=BMM_Additive
    scaleSpace=BCS_BoneSpace
    translationMode=BMM_Additive
    translationSpace=BCS_BoneSpace
    alphaBoolBlend {blendInTime=0.2, blendOutTime=0.2}
    boneToModify {boneName=base_ponytail_l}
    rotation=<rot 0 0 -71.348>
   }
   n4 = "Transform (Modify) Bone - Bone: base_ponytail_r"!(ComponentPose=@n3, bAlphaBoolEnabled:bool=True, Translation="0,0,0", Rotation="0.000,0.000,90.128", Scale="1,1,1") -> (Pose) {
    alphaInputType=Bool
    rotationMode=BMM_Additive
    rotationSpace=BCS_BoneSpace
    scaleMode=BMM_Additive
    scaleSpace=BCS_BoneSpace
    translationMode=BMM_Additive
    translationSpace=BCS_BoneSpace
    alphaBoolBlend {blendInTime=0.2, blendOutTime=0.2}
    boneToModify {boneName=base_ponytail_r}
    rotation=<rot 0 0 90.128>
   }
   n5 = RigidBody!(
    ComponentPose=@n4
    Alpha
    OverrideWorldGravity:Vector="0,0,0"
    ExternalForce
    SimSpaceSettings:Sim Space Settings Structure=(WorldAlpha=0,VelocityScaleZ=1,DampingAlpha=1,MaxLinearVelocity=10000,MaxAngularVelocity=10000,MaxLinearAcceleration=10000,MaxAngularAcceleration=10000,ExternalLinearDragV=(X=0,Y=0,Z=0),ExternalLinearVelocity=(X=0,Y=0,Z=0),ExternalAngularVelocity=(X=0,Y=0,Z=0))
   ) -> (Pose) {
    bOverrideWorldGravity=true
    overridePhysicsAsset=&/F_MED_WaltzScout_FaceAcc_Physics
    simulationSpace=BaseBoneSpace
    baseBoneRef {boneName=head}
    componentAppliedLinearAccClamp=<3000 3000 3000>
   }
   n6 = Anim Dynamics - Bone: dyn_hair_side_01_l(ComponentPose=@n5, Alpha=1, ExternalForce="0,0,0") {
    angularDampingOverride=1
    angularSpringConstant=25000
    bAngularSpring=true
    bOverrideAngularDamping=true
    bUseGravityOverride=true
    gravityScale=0.35
    simulationSpace=BoneRelative
    boundBone {boneName=dyn_hair_side_01_l}
    componentAppliedLinearAccClamp=<3000 3000 3000>
    componentLinearAccScale=<1 1 1>
    componentLinearVelScale=<1 1 1>
    gravityOverride {z=-980}
    physicsBodyDefinitions [{
      boundBone {boneName=dyn_hair_side_01_l}
      boxExtents=<20 40 30>
      constraintSetup {angularLimitsMax=<10 -4 15>, angularLimitsMin=<-10 -10 0>, angularTarget=<1 0 0>}
      localJointOffset=<7 0 0>
     }]
    relativeSpaceBone {boneName=root}
   }
   n7 = Anim Dynamics - Bone: dyn_hair_side_02_l(ComponentPose=@n6, Alpha=1, ExternalForce="0,0,0") {
    angularDampingOverride=1
    angularSpringConstant=25000
    bAngularSpring=true
    bOverrideAngularDamping=true
    bUseGravityOverride=true
    gravityScale=0.35
    simulationSpace=BoneRelative
    boundBone {boneName=dyn_hair_side_02_l}
    componentAppliedLinearAccClamp=<3000 3000 3000>
    componentLinearAccScale=<1 1 1>
    componentLinearVelScale=<1 1 1>
    gravityOverride {z=-980}
    physicsBodyDefinitions [{
      boundBone {boneName=dyn_hair_side_02_l}
      boxExtents=<20 40 30>
      constraintSetup {angularLimitsMax=<10 -4 15>, angularLimitsMin=<-10 -10 0>, angularTarget=<1 0 0>}
      localJointOffset=<7 0 0>
     }]
    relativeSpaceBone {boneName=root}
   }
   n8 = Anim Dynamics - Bone: dyn_hair_side_03_l(ComponentPose=@n7, Alpha=1, ExternalForce="0,0,0") {
    angularDampingOverride=1
    angularSpringConstant=25000
    bAngularSpring=true
    bOverrideAngularDamping=true
    bUseGravityOverride=true
    gravityScale=0.35
    simulationSpace=BoneRelative
    boundBone {boneName=dyn_hair_side_03_l}
    componentAppliedLinearAccClamp=<3000 3000 3000>
    componentLinearAccScale=<1 1 1>
    componentLinearVelScale=<1 1 1>
    gravityOverride {z=-980}
    physicsBodyDefinitions [{
      boundBone {boneName=dyn_hair_side_03_l}
      boxExtents=<20 40 30>
      constraintSetup {angularLimitsMax=<10 -4 15>, angularLimitsMin=<-10 -10 0>, angularTarget=<1 0 0>}
      localJointOffset=<7 0 0>
     }]
    relativeSpaceBone {boneName=root}
   }
   n9 = Anim Dynamics - Bone: dyn_hair_side_01_r(ComponentPose=@n8, Alpha=1, ExternalForce="0,0,0") {
    angularDampingOverride=0.99
    angularSpringConstant=25000
    bAngularSpring=true
    bOverrideAngularDamping=true
    bUseGravityOverride=true
    gravityScale=0.35
    simulationSpace=BoneRelative
    boundBone {boneName=dyn_hair_side_01_r}
    componentAppliedLinearAccClamp=<3000 3000 3000>
    componentLinearAccScale=<1 1 1>
    componentLinearVelScale=<1 1 1>
    gravityOverride {z=-980}
    physicsBodyDefinitions [{
      boundBone {boneName=dyn_hair_side_01_r}
      boxExtents=<30 40 30>
      constraintSetup {angularLimitsMax=<10 1 10>, angularLimitsMin=<-10 -5 0>, angularTarget=<1 0 0>}
      localJointOffset=<-7 0 0>
     }]
    relativeSpaceBone {boneName=root}
    simSpaceSettings {maxAngularAcceleration=500, maxAngularVelocity=500, simSpaceAngularAlpha=0.5}
   }
   n10 = Anim Dynamics - Bone: dyn_hair_side_02_r(ComponentPose=@n9, Alpha=1, ExternalForce="0,0,0") {
    angularDampingOverride=0.99
    angularSpringConstant=25000
    bAngularSpring=true
    bOverrideAngularDamping=true
    bUseGravityOverride=true
    gravityScale=0.35
    simulationSpace=BoneRelative
    boundBone {boneName=dyn_hair_side_02_r}
    componentAppliedLinearAccClamp=<3000 3000 3000>
    componentLinearAccScale=<1 1 1>
    componentLinearVelScale=<1 1 1>
    gravityOverride {z=-980}
    physicsBodyDefinitions [{
      boundBone {boneName=dyn_hair_side_02_r}
      boxExtents=<30 40 30>
      constraintSetup {angularLimitsMax=<10 1 10>, angularLimitsMin=<-10 -5 0>, angularTarget=<1 0 0>}
      localJointOffset=<-7 0 0>
     }]
    relativeSpaceBone {boneName=root}
    simSpaceSettings {maxAngularAcceleration=500, maxAngularVelocity=500, simSpaceAngularAlpha=0.5}
   }
   n11 = Anim Dynamics - Bone: dyn_hair_side_03_r(ComponentPose=@n10, Alpha=1, ExternalForce="0,0,0") {
    angularDampingOverride=0.99
    angularSpringConstant=25000
    bAngularSpring=true
    bOverrideAngularDamping=true
    bUseGravityOverride=true
    gravityScale=0.35
    simulationSpace=BoneRelative
    boundBone {boneName=dyn_hair_side_03_r}
    componentAppliedLinearAccClamp=<3000 3000 3000>
    componentLinearAccScale=<1 1 1>
    componentLinearVelScale=<1 1 1>
    gravityOverride {z=-980}
    physicsBodyDefinitions [{
      boundBone {boneName=dyn_hair_side_03_r}
      boxExtents=<30 40 30>
      constraintSetup {angularLimitsMax=<10 1 10>, angularLimitsMin=<-10 -5 0>, angularTarget=<1 0 0>}
      localJointOffset=<-7 0 0>
     }]
    relativeSpaceBone {boneName=root}
    simSpaceSettings {maxAngularAcceleration=500, maxAngularVelocity=500, simSpaceAngularAlpha=0.5}
   }
   Save cached pose 'OuputAnimGraph'(Pose=Component To Local(ComponentPose=@n11)) {cacheName=OuputAnimGraph}
  ]}
 EventGraph {nodes [
   Event OnCharacterCustomizationLoaded()
   > n1 = Cast To PlayerPawn_Athena(Object=TryGetPawnOwner())
   > Set Body_Mesh(Body_Mesh=GetSkeletalMeshForPartType(self=@n1))
   Event BlueprintUpdateAnimation()
   > n2 = Sequence()
   > n3 = "Gravity Override State Function (Parameter Asset) (Nativized)"(
    CharacterPartAnimInstance=Self-Reference()
    GameTimeInSeconds=GetGameTimeInSeconds()
    DebugJointName=spine_05
    Parameters=&/F_MED_WaltzScout_FaceAccParameters
   )
   > Set Gravity Override(Gravity Override=@n3)
   > Set RBAN Settings Out(RBAN Settings Out=@n3.OutRigidBodyAnimNodeParameters)
   Branch(execute=@n2.then_1, Condition=Get bIsInFrontEnd())
   > Branch(Condition=Get bIsInFrontEnd())
   > Set Custom Idle Is Playing Float(Custom Idle Is Playing Float:double=GetCurveValue(self=GetOwnerFrontEndAnimInstance(), CurveName=Frontend_Idle_Bypass)) -> (Output_Get:double)
  ]}
]
variables [
 Body_Mesh {type=SkeletalMeshComponent}
 Custom Idle Is Playing Float {type=number}
 Gravity Override {type=Vector}
 RBAN Settings Out {type=FortRigidBodyAnimNodeParameters}
]
